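# Physics asset for the TalonHime medium female player body, covering its hanging cloth bones and the lower body they can collide with.
# A front chain hangs from a master bone locked to the pelvis, and left and right back chains attach to the pelvis with twist locked, each using per-link angular limits.
# Capsules cover pelvis, legs and right foot only, plus a huge root capsule below the origin that resembles a ground collider.
# The front master-to-pelvis constraint appears twice.
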PhysicsAsset /Game/Characters/Player/Female/Medium/Bodies/F_MED_TalonHime/Meshes/F_MED_TalonHime_Physics
bodies {
 calf_l [{shapeType=Capsule, length=10.4115, radius=6.3318, center=<-16.1634 -2.6044 -0.6519>, rotation=<rot -0.0459 87.5248 -89.225>}]
 calf_r [{shapeType=Capsule, length=10.4115, radius=6.3318, center=<16.1634 2.6044 0.6519>, rotation=<rot -0.0459 87.5248 90.775>}]
 calf_twist_01_l [{shapeType=Capsule, length=13.4115, radius=3.3318, center=<0.708 -0.1857 -0.4278>, rotation=<rot -0.0459 87.5248 -89.225>}]
 calf_twist_01_r [{shapeType=Capsule, length=13.4115, radius=3.3318, center=<-0.708 0.1857 0.4278>, rotation=<rot -0.0459 87.5248 90.775>}]
 deform_knee_in_l [{shapeType=Capsule, length=0, radius=7.2, shapeName=deform_knee_in_l_capsule, center=<-5.0501 4.9907 0>}]
 deform_knee_in_r [{shapeType=Capsule, length=0, radius=7.2, shapeName=deform_knee_in_l_capsule, center=<5.0501 -4.9907 0>, rotation=<rot 0 0 -180>}]
 dyn_cloth_back_00_l [{
   shapeType=Capsule
   length=0.5
   radius=7.7
   shapeName=dyn_cloth_back_00_l_capsule
   center=<-7.5923 -4.8119 -3.1462>
   rotation=<rot 83.9589 179.9999 -180>
  }]
 dyn_cloth_back_00_r [{
   shapeType=Capsule
   length=0.5
   radius=7.7
   shapeName=dyn_cloth_back_00_l_capsule
   center=<7.5923 4.8119 3.1462>
   rotation=<rot 83.9589 179.9999 0>
  }]
 dyn_cloth_back_01_l [{
   shapeType=Capsule
   length=0
   radius=7.2
   shapeName=dyn_cloth_back_01_l_capsule
   center=<-6.6692 -2.7048 -1.628>
   rotation=<rot 83.9589 179.9999 179.9999>
  }]
 dyn_cloth_back_01_r [{
   shapeType=Capsule
   length=0
   radius=7.2
   shapeName=dyn_cloth_back_01_l_capsule
   center=<6.6692 2.7048 1.628>
   rotation=<rot 83.9589 179.9999 -0.0001>
  }]
 dyn_cloth_back_02_l [{
   shapeType=Capsule
   length=0
   radius=6.2
   shapeName=dyn_cloth_back_01_l_capsule
   center=<-6.8905 -2.12 0.4634>
   rotation=<rot 83.9589 179.9999 179.9999>
  }]
 dyn_cloth_back_02_r [{
   shapeType=Capsule
   length=0
   radius=6.2
   shapeName=dyn_cloth_back_01_l_capsule
   center=<6.8905 2.12 -0.4634>
   rotation=<rot 83.9589 179.9999 -0.0001>
  }]
 dyn_cloth_back_03_l [{
   shapeType=Capsule
   length=4.5
   radius=4.7
   shapeName=dyn_cloth_back_03_l_capsule
   center=<-6.8905 -2.12 0.4634>
   rotation=<rot 83.9589 179.9999 180>
  }]
 dyn_cloth_back_03_r [{
   shapeType=Capsule
   length=4.5
   radius=4.7
   shapeName=dyn_cloth_back_03_l_capsule
   center=<6.8905 2.12 -0.4634>
   rotation=<rot 83.9589 179.9999 0>
  }]
 dyn_cloth_back_04_l [{
   shapeType=Capsule
   length=4.9
   radius=5.4
   shapeName=dyn_cloth_back_04_l_capsule
   center=<-7.8771 -1.1783 0.359>
   rotation=<rot 83.9589 -180 -179.9999>
  }]
 dyn_cloth_back_04_r [{
   shapeType=Capsule
   length=4.9
   radius=5.4
   shapeName=dyn_cloth_back_04_l_capsule
   center=<7.8771 1.1783 -0.359>
   rotation=<rot 83.9589 -180 0.0001>
  }]
 dyn_cloth_front_00 [{shapeType=Capsule, length=0, radius=9, shapeName=dyn_cloth_front_03_capsule, center=<3.334 -6.6735 0>, rotation=<rot 90 0 0>}]
 dyn_cloth_front_01 [{
   shapeType=Capsule
   length=0
   radius=8.9
   shapeName=dyn_cloth_front_03_capsule
   center=<3.4667 -6.7333 0.6831>
   rotation=<rot 90 180 -179.114>
  }]
 dyn_cloth_front_02 [{shapeType=Capsule, length=0, radius=5.9, shapeName=dyn_cloth_front_03_capsule, center=<5.8939 -5.9911 0>, rotation=<rot 90 0 0>}]
 dyn_cloth_front_03 [{shapeType=Capsule, length=0, radius=5.8, shapeName=dyn_cloth_front_03_capsule, center=<7.1431 -4.8556 0>, rotation=<rot 90 0 0>}]
 dyn_cloth_front_04 [{shapeType=Capsule, length=3.5, radius=4.8, shapeName=dyn_cloth_front_03_capsule, center=<7.4446 -1.7868 0>, rotation=<rot 90 0 0>}]
 dyn_cloth_front_MSTR [{shapeType=Capsule, length=0.5, radius=0.5, shapeName=dyn_cloth_front_MSTR_capsule}]
 foot_r [{shapeType=Capsule, length=8.4, radius=5.8, shapeName=foot_r_capsule, center=<3.1454 -5.2183 0.4304>, rotation=<rot 0 -0.0001 -85.2852>}]
 pelvis [{shapeType=Capsule, length=7.6853, radius=9.7427, center=<-4.0327 -3.9639 0.22>, rotation=<rot 0 90 0>}]
 root [{shapeType=Capsule, length=62.6, radius=226.4999, shapeName=root_capsule, center=<0 -9.8986 -224.6921>, rotation=<rot 90 180 90>}]
 thigh_l [{shapeType=Capsule, length=10.4808, radius=8.0629, center=<-30.0778 1.3929 1.7145>, rotation=<rot -0.241 -91.6384 -84.3291>}]
 thigh_r [{shapeType=Capsule, length=10.4808, radius=8.0629, center=<30.1472 -1.3929 -1.7043>, rotation=<rot -0.241 -91.6384 95.6709>}]
 thigh_twist_01_l [{
   shapeType=Capsule
   length=14.8
   radius=9.5
   shapeName=thigh_twist_01_l_capsule
   center=<14.8917 -0.9824 0.5384>
   rotation=<rot 82.4098 175.9918 179.9997>
  }]
 thigh_twist_01_r [{
   shapeType=Capsule
   length=15.3
   radius=9.5
   shapeName=thigh_twist_01_l_capsule
   center=<-14.8917 0.9824 -0.5384>
   rotation=<rot 82.4098 175.9918 179.9997>
  }]
}
constraints [
 {bone1Name=dyn_cloth_front_03, bone2Name=dyn_cloth_front_02, swing1LimitDegrees=8, swing2LimitDegrees=30, twistLimitDegrees=3}
 {bone1Name=dyn_cloth_back_04_l, bone2Name=dyn_cloth_back_03_l, swing1LimitDegrees=45, swing2LimitDegrees=45, twistLimitDegrees=25}
 {bone1Name=dyn_cloth_back_03_l, bone2Name=dyn_cloth_back_02_l, swing1LimitDegrees=45, swing2LimitDegrees=45, twistLimitDegrees=25}
 {bone1Name=dyn_cloth_back_02_l, bone2Name=dyn_cloth_back_01_l, swing1LimitDegrees=45, swing2LimitDegrees=45, twistLimitDegrees=20}
 {bone1Name=dyn_cloth_back_01_l, bone2Name=dyn_cloth_back_00_l, swing1LimitDegrees=30, swing2LimitDegrees=45, twistLimitDegrees=3}
 {bone1Name=dyn_cloth_back_00_l, bone2Name=pelvis, swing1LimitDegrees=20, swing2LimitDegrees=25, twistLimitDegrees=45, twistMotion=Locked}
 {
  bone1Name=calf_r
  bone2Name=thigh_r
  swing1LimitDegrees=45
  swing1Motion=Free
  swing2LimitDegrees=45
  swing2Motion=Free
  twistLimitDegrees=45
  twistMotion=Free
 }
 {
  bone1Name=deform_knee_in_r
  bone2Name=calf_r
  swing1LimitDegrees=45
  swing1Motion=Free
  swing2LimitDegrees=45
  swing2Motion=Free
  twistLimitDegrees=45
  twistMotion=Free
 }
 {
  bone1Name=thigh_r
  bone2Name=pelvis
  swing1LimitDegrees=45
  swing1Motion=Free
  swing2LimitDegrees=45
  swing2Motion=Free
  twistLimitDegrees=45
  twistMotion=Free
 }
 {
  bone1Name=thigh_twist_01_r
  bone2Name=thigh_r
  swing1LimitDegrees=45
  swing1Motion=Free
  swing2LimitDegrees=45
  swing2Motion=Free
  twistLimitDegrees=45
  twistMotion=Free
 }
 {bone1Name=dyn_cloth_back_04_r, bone2Name=dyn_cloth_back_03_r, swing1LimitDegrees=45, swing2LimitDegrees=45, twistLimitDegrees=25}
 {bone1Name=dyn_cloth_back_03_r, bone2Name=dyn_cloth_back_02_r, swing1LimitDegrees=45, swing2LimitDegrees=45, twistLimitDegrees=25}
 {bone1Name=dyn_cloth_back_02_r, bone2Name=dyn_cloth_back_01_r, swing1LimitDegrees=45, swing2LimitDegrees=45, twistLimitDegrees=20}
 {bone1Name=dyn_cloth_back_01_r, bone2Name=dyn_cloth_back_00_r, swing1LimitDegrees=30, swing2LimitDegrees=45, twistLimitDegrees=3}
 {bone1Name=dyn_cloth_back_00_r, bone2Name=pelvis, swing1LimitDegrees=20, swing2LimitDegrees=25, twistLimitDegrees=45, twistMotion=Locked}
 {
  bone1Name=calf_twist_01_l
  bone2Name=calf_l
  swing1LimitDegrees=45
  swing1Motion=Free
  swing2LimitDegrees=45
  swing2Motion=Free
  twistLimitDegrees=45
  twistMotion=Free
 }
 {bone1Name=dyn_cloth_front_00, bone2Name=dyn_cloth_front_MSTR, swing1LimitDegrees=80, swing2LimitDegrees=30, twistLimitDegrees=10}
 {
  bone1Name=dyn_cloth_front_MSTR
  bone2Name=pelvis
  swing1LimitDegrees=45
  swing1Motion=Locked
  swing2LimitDegrees=45
  swing2Motion=Locked
  twistLimitDegrees=45
  twistMotion=Locked
 }
 {bone1Name=dyn_cloth_front_04, bone2Name=dyn_cloth_front_03, swing1LimitDegrees=10, swing2LimitDegrees=30, twistLimitDegrees=3}
 {
  bone1Name=dyn_cloth_front_MSTR
  bone2Name=pelvis
  swing1LimitDegrees=45
  swing1Motion=Locked
  swing2LimitDegrees=45
  swing2Motion=Locked
  twistLimitDegrees=45
  twistMotion=Locked
 }
 {bone1Name=dyn_cloth_front_01, bone2Name=dyn_cloth_front_00, swing1LimitDegrees=80, swing2LimitDegrees=30, twistLimitDegrees=30}
 {bone1Name=dyn_cloth_front_02, bone2Name=dyn_cloth_front_01, swing1LimitDegrees=8, swing2LimitDegrees=30, twistLimitDegrees=10}
]
