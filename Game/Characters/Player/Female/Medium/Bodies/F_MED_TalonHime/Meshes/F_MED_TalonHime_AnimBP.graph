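# Animation Blueprint for the female medium TalonHime player body, driving secondary motion on the front and back cloth bones.
# AnimDynamics runs from the cached PreRigid pose, with bone-relative chains weighted by a skydiving alpha and world-space per-bone simulations by on-ground alphas.
# Planar and spherical limits on leg, pelvis and root bones constrain the cloth.
# Bool-blended Transform Bone nodes add corrective offsets for states like melee, sliding, tactical sprint and sprint jumping.
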AnimBlueprint /Game/Characters/Player/Female/Medium/Bodies/F_MED_TalonHime/Meshes/F_MED_TalonHime_AnimBP
target_skeleton=&/F_MED_TalonHime_Skeleton
graphs [
 AnimGraph {nodes [
   "Anim Dynamics - Chain: dyn_cloth_front_00 -> dyn_cloth_front_04"!(ComponentPose, Alpha=1, GravityOverride:Vector="0,0,300", ExternalForce="0,0,0") -> (Pose) {
    angularDampingOverride=1
    angularSpringConstant=3000
    bAngularSpring=true
    bChain=true
    bEnableWind=true
    bOverrideAngularDamping=true
    bUseGravityOverride=true
    simulationSpace=BoneRelative
    windScale=100
    alphaBoolBlend {blendOption=Linear}
    alphaScaleBiasClamp {bInterpResult=true}
    boundBone {boneName=dyn_cloth_front_00}
    chainEnd {boneName=dyn_cloth_front_04}
    componentAppliedLinearAccClamp=<150 150 180>
    componentLinearAccScale=<0.1 0.1 0.2>
    componentLinearVelScale=<0.3 0.3 0.4>
    gravityOverride {z=-700}
    physicsBodyDefinitions [
     {
      collisionType=CustomSphere
      sphereCollisionRadius=1
      boundBone {boneName=dyn_cloth_front_00}
      boxExtents=<2 10 10>
      constraintSetup {angularLimitsMax=<5 20 40>, angularLimitsMin=<-5 -20 -40>, angularTarget=<0 0 0>}
      localJointOffset=<1 0 0>
     }
     {
      collisionType=CustomSphere
      sphereCollisionRadius=1
      boundBone {boneName=dyn_cloth_front_01}
      boxExtents=<2 10 10>
      constraintSetup {angularLimitsMax=<5 20 40>, angularLimitsMin=<-5 -20 -40>, angularTarget=<0 0 0>}
      localJointOffset=<1 0 0>
     }
     {
      collisionType=CustomSphere
      sphereCollisionRadius=1
      boundBone {boneName=dyn_cloth_front_02}
      boxExtents=<2 10 10>
      constraintSetup {angularLimitsMax=<5 20 40>, angularLimitsMin=<-5 -20 -40>, angularTarget=<0 0 0>}
      localJointOffset=<1 0 0>
     }
     {
      collisionType=CustomSphere
      sphereCollisionRadius=1
      boundBone {boneName=dyn_cloth_front_03}
      boxExtents=<2 10 10>
      constraintSetup {angularLimitsMax=<5 20 40>, angularLimitsMin=<-5 -20 -40>, angularTarget=<0 0 0>}
      localJointOffset=<1 0 0>
     }
     {
      collisionType=CustomSphere
      sphereCollisionRadius=1
      boundBone {boneName=dyn_cloth_front_04}
      boxExtents=<2 10 10>
      constraintSetup {angularLimitsMax=<5 20 40>, angularLimitsMin=<-5 -20 -40>, angularTarget=<0 0 0>}
      localJointOffset=<1 0 0>
     }
    ]
    planarLimits [
     {drivingBone {boneName=thigh_l}, planeTransform {location=<0 0 6>, rotation=<rot -7.0001 0 0>}}
     {drivingBone {boneName=thigh_r}, planeTransform {location=<0 0 -6>, rotation=<rot 7.0001 180 180>}}
     {drivingBone {boneName=pelvis}, planeTransform {location=<30 3 0>, rotation=<rot 0 0 90>, scale=<0.5 0.5 0.5>}}
     {drivingBone {boneName=root}, planeTransform {location=<0 0 0>, rotation=<rot 0 0 0>}}
    ]
    relativeSpaceBone {boneName=root}
   }
   Look At - Bone: dyn_cloth_front_00!(ComponentPose, bAlphaBoolEnabled:bool=True) -> (Pose) {
    alphaInputType=Bool
    alphaBoolBlend {blendInTime=0.5, blendOption=ExpInOut, blendOutTime=0.5}
    boneToModify {boneName=dyn_cloth_front_00}
    lookAtLocation=<-27.7324 -0.8735 0.1079>
    lookAtTarget {boneReference {boneName=pelvis}}
    lookAt_Axis {axis=<1 0>}
   }
   n1 = Look At - Bone: dyn_cloth_back_00_r!(ComponentPose, bAlphaBoolEnabled:bool=True) -> (Pose) {
    alphaInputType=Bool
    alphaBoolBlend {blendInTime=0.5, blendOption=ExpInOut, blendOutTime=0.5}
    boneToModify {boneName=dyn_cloth_back_00_r}
    lookAtLocation=<-0.9247 -19.3823 1.3984>
    lookAtTarget {boneReference {boneName=pelvis}}
    lookAt_Axis {axis=<1 0>}
   }
   Look At - Bone: dyn_cloth_back_00_l!(ComponentPose=@n1, bAlphaBoolEnabled:bool=True) -> (Pose) {
    alphaInputType=Bool
    alphaBoolBlend {blendInTime=0.5, blendOption=ExpInOut, blendOutTime=0.5}
    boneToModify {boneName=dyn_cloth_back_00_l}
    lookAtLocation=<1.441 -22.9797 -3.9871>
    lookAtTarget {boneReference {boneName=pelvis}}
    lookAt_Axis {axis=<-1 0>}
   }
   n2 = Get bIsSliding()
   "Transform (Modify) Bone - Bone: dyn_cloth_front_MSTR"!(ComponentPose, bAlphaBoolEnabled:bool=@n2, Translation="-2.563,29.614,2.863", Rotation="0,0,0", Scale="1,1,1") -> (Pose) {
    alphaInputType=Bool
    rotationMode=BMM_Replace
    rotationSpace=BCS_BoneSpace
    translationMode=BMM_Additive
    translationSpace=BCS_BoneSpace
    alphaBoolBlend {blendInTime=0.3, blendOption=ExpInOut, blendOutTime=0.3}
    boneToModify {boneName=dyn_cloth_front_MSTR}
    translation=<-2.563 29.614 2.863>
   }
   n3 = "Get Walking/running  alpha"() -> ("Walking/running  alpha":bool)
   n4 = "Transform (Modify) Bone - Bone: dyn_cloth_front_MSTR"!(
    ComponentPose
    bAlphaBoolEnabled=@n3
    Translation=Get Front cloth corr() -> (Front cloth corr:Vector)
    Rotation_Roll:float=0
    Rotation_Pitch:float=0
    Rotation_Yaw:float=Get Front cloth rot corr() -> (Front cloth rot corr:double)
    Scale="1,1,1"
   ) -> (Pose) {
    alphaInputType=Bool
    rotationMode=BMM_Additive
    rotationSpace=BCS_BoneSpace
    translationMode=BMM_Additive
    translationSpace=BCS_BoneSpace
    alphaBoolBlend {blendInTime=0.3, blendOption=Linear, blendOutTime=0.3}
    boneToModify {boneName=dyn_cloth_front_MSTR}
    rotation {yaw=-20}
    translation=<-1 -1 -0.5>
   }
   n5 = "Transform (Modify) Bone - Bone: dyn_cloth_front_00"!(ComponentPose=@n4, bAlphaBoolEnabled:bool=@n3, Translation="0,0,0", Rotation="0,-15,0", Scale="1,1,1") -> (Pose) {
    alphaInputType=Bool
    rotationMode=BMM_Additive
    rotationSpace=BCS_BoneSpace
    translationMode=BMM_Additive
    translationSpace=BCS_BoneSpace
    alphaBoolBlend {blendInTime=0.3, blendOption=Linear, blendOutTime=0.3}
    boneToModify {boneName=dyn_cloth_front_00}
    rotation {yaw=-15}
   }
   n6 = "Transform (Modify) Bone - Bone: dyn_cloth_front_MSTR"!(
    ComponentPose=@n5
    bAlphaBoolEnabled=Get Crouching idle alpha() -> (Crouching idle alpha:bool)
    Translation="0,-3,2"
    Rotation:Rotator="0,-20,5"
    Scale="1,1,1"
   ) -> (Pose) {
    alphaInputType=Bool
    rotationMode=BMM_Additive
    rotationSpace=BCS_BoneSpace
    translationMode=BMM_Additive
    translationSpace=BCS_BoneSpace
    alphaBoolBlend {blendInTime=0.3, blendOption=Linear, blendOutTime=0.3}
    boneToModify {boneName=dyn_cloth_front_MSTR}
    rotation {roll=5, yaw=-20}
    translation {y=-3, z=2}
   }
   n7 = "Transform (Modify) Bone - Bone: dyn_cloth_front_MSTR"(ComponentPose=@n6, bAlphaBoolEnabled=Get bIsPlayingEmote(), Translation="-1,-2,0", Rotation="0,-5,0", Scale="1,1,1") {
    alphaInputType=Bool
    rotationMode=BMM_Additive
    rotationSpace=BCS_BoneSpace
    translationMode=BMM_Additive
    translationSpace=BCS_BoneSpace
    alphaBoolBlend {blendInTime=0.3, blendOption=Linear, blendOutTime=0.8}
    boneToModify {boneName=dyn_cloth_front_MSTR}
    rotation {yaw=-5}
    translation=<-1 -2>
   }
   n8 = "Transform (Modify) Bone - Bone: dyn_cloth_front_MSTR"(ComponentPose=@n7, bAlphaBoolEnabled=Get bIsPlayingMeleeAnim(), Translation="0,-2.5,0.8", Rotation="0,0,0", Scale="1,1,1") {
    alphaInputType=Bool
    rotationSpace=BCS_BoneSpace
    translationMode=BMM_Additive
    translationSpace=BCS_BoneSpace
    alphaBoolBlend {blendInTime=0.3, blendOption=Linear, blendOutTime=0.8}
    boneToModify {boneName=dyn_cloth_front_MSTR}
    translation {y=-2.5, z=0.8}
   }
   n9 = "Transform (Modify) Bone - Bone: dyn_cloth_front_MSTR"(
    ComponentPose=@n8
    bAlphaBoolEnabled=Get Is melee while crouch() -> (Is melee while crouch:bool)
    Translation="0,-1.5,0"
    Rotation="0,0,0"
    Scale="1,1,1"
   ) {
    alphaInputType=Bool
    rotationMode=BMM_Additive
    rotationSpace=BCS_BoneSpace
    translationMode=BMM_Additive
    translationSpace=BCS_BoneSpace
    alphaBoolBlend {blendInTime=0.3, blendOption=Linear, blendOutTime=0.2}
    boneToModify {boneName=dyn_cloth_front_MSTR}
    translation {y=-1.5}
   }
   n10 = "Constraint - dyn_cloth_front_00 (PP)"!(ComponentPose=@n9, Alpha=1, ConstraintWeights_0=0.5, ConstraintWeights_1:float=0.5) -> (Pose) {
    alphaBoolBlend {blendOption=Linear}
    boneToModify {boneName=dyn_cloth_front_00}
    constraintSetup [{targetBone {boneName=thigh_twist_01_l}}, {targetBone {boneName=thigh_twist_01_r}}]
    constraintWeights [1,1]
   }
   n11 = "Transform (Modify) Bone - Bone: dyn_cloth_front_MSTR"(ComponentPose=@n10, bAlphaBoolEnabled=Get bIsMoving2D(), Translation="0,-3,0", Rotation="0,-5,0", Scale="1,1,1") {
    alphaInputType=Bool
    rotationMode=BMM_Additive
    rotationSpace=BCS_BoneSpace
    translationMode=BMM_Additive
    translationSpace=BCS_BoneSpace
    alphaBoolBlend {blendInTime=0.3, blendOption=Linear, blendOutTime=0.2}
    boneToModify {boneName=dyn_cloth_front_MSTR}
    rotation {yaw=-5}
    translation {y=-3}
   }
   "Transform (Modify) Bone - Bone: dyn_cloth_front_MSTR"(
    ComponentPose=@n11
    bAlphaBoolEnabled=Get Crouching corr() -> (Crouching corr:bool)
    Translation="-4,-25,0"
    Rotation="0,-20,0"
    Scale="1,1,1"
   ) {
    alphaInputType=Bool
    rotationMode=BMM_Additive
    rotationSpace=BCS_BoneSpace
    translationMode=BMM_Additive
    translationSpace=BCS_BoneSpace
    alphaBoolBlend {blendOption=Linear}
    boneToModify {boneName=dyn_cloth_front_MSTR}
    rotation {yaw=-20}
    translation=<-4 -25>
   }
   n12 = "Transform (Modify) Bone - Bone: dyn_cloth_back_03_r"!(ComponentPose, bAlphaBoolEnabled:bool=True, Translation="0,0,0", Rotation="0,-20,0", Scale="1,1,1") -> (Pose) {
    alphaInputType=Bool
    rotationMode=BMM_Additive
    rotationSpace=BCS_BoneSpace
    translationMode=BMM_Additive
    translationSpace=BCS_BoneSpace
    alphaBoolBlend {blendInTime=0.3, blendOption=ExpInOut, blendOutTime=0.3}
    boneToModify {boneName=dyn_cloth_back_03_r}
    rotation {yaw=-20}
   }
   "Transform (Modify) Bone - Bone: dyn_cloth_back_03_l"!(ComponentPose=@n12, bAlphaBoolEnabled:bool=True, Translation="0,0,0", Rotation="0,-10,0", Scale="1,1,1") -> (Pose) {
    alphaInputType=Bool
    rotationMode=BMM_Additive
    rotationSpace=BCS_BoneSpace
    translationMode=BMM_Additive
    translationSpace=BCS_BoneSpace
    alphaBoolBlend {blendInTime=0.3, blendOption=ExpInOut, blendOutTime=0.3}
    boneToModify {boneName=dyn_cloth_back_03_l}
    rotation {yaw=-10}
   }
   n13 = "Transform (Modify) Bone - Bone: dyn_cloth_back_02_r"!(ComponentPose, bAlphaBoolEnabled:bool=True, Translation="0,0,0", Rotation="0,10,0", Scale="1,1,1") -> (Pose) {
    alphaInputType=Bool
    rotationMode=BMM_Additive
    rotationSpace=BCS_BoneSpace
    translationMode=BMM_Additive
    translationSpace=BCS_BoneSpace
    alphaBoolBlend {blendInTime=0.3, blendOption=ExpInOut, blendOutTime=0.3}
    boneToModify {boneName=dyn_cloth_back_02_r}
    rotation {yaw=10}
   }
   "Transform (Modify) Bone - Bone: dyn_cloth_back_02_l"!(ComponentPose=@n13, bAlphaBoolEnabled:bool=True, Translation="0,0,0", Rotation="0,5,0", Scale="1,1,1") -> (Pose) {
    alphaInputType=Bool
    rotationMode=BMM_Additive
    rotationSpace=BCS_BoneSpace
    translationMode=BMM_Additive
    translationSpace=BCS_BoneSpace
    alphaBoolBlend {blendInTime=0.3, blendOption=ExpInOut, blendOutTime=0.3}
    boneToModify {boneName=dyn_cloth_back_02_l}
    rotation {yaw=5}
   }
   "Transform (Modify) Bone - Bone: shoulder_pad_r"!(
    ComponentPose
    bAlphaBoolEnabled:bool=Get bIsPlayingMeleeAnim()
    Translation="-0.427,1.214,0.603"
    Rotation="-0.798,3.873,-15.462"
    Scale="1,1,1"
   ) -> (Pose) {
    alphaInputType=Bool
    rotationMode=BMM_Additive
    rotationSpace=BCS_BoneSpace
    translationMode=BMM_Additive
    translationSpace=BCS_BoneSpace
    alphaBoolBlend {blendInTime=0.3, blendOption=Linear, blendOutTime=0.5}
    boneToModify {boneName=shoulder_pad_r}
    rotation=<rot -0.798 3.873 -15.462>
    translation=<-0.427 1.214 0.6028>
   }
   n14 = Get bIsParachuteOpened()
   n15 = Get bIsSkydiveFloating()
   n16 = Get bIsSkydiveDiveMode()
   n17 = "Get Is Sprint Jumping "() -> ("Is Sprint Jumping ":bool)
   n18 = Get Skydiving Anim Dynamics Alpha() -> (Skydiving Anim Dynamics Alpha:double)
   n19 = "Anim Dynamics - Chain: dyn_cloth_back_00_r -> dyn_cloth_back_04_r"!(
    ComponentPose=Local To Component(LocalPose=Use cached pose 'PreRigid'())
    Alpha=@n18
    GravityOverride:Vector="0,0,300"
    ExternalForce="0,0,0"
   ) -> (Pose) {
    angularDampingOverride=1
    angularSpringConstant=10000
    bAngularSpring=true
    bChain=true
    bOverrideAngularDamping=true
    bUseGravityOverride=true
    simulationSpace=BoneRelative
    alphaBoolBlend {blendOption=Linear}
    alphaScaleBiasClamp {bInterpResult=true}
    boundBone {boneName=dyn_cloth_back_00_r}
    chainEnd {boneName=dyn_cloth_back_04_r}
    componentAppliedLinearAccClamp=<1000 1000 1000>
    componentLinearAccScale=<0.0001 0.0001 0.0001>
    componentLinearVelScale=<0.0001 0.0001 0.0001>
    gravityOverride {z=-700}
    physicsBodyDefinitions [
     {
      boundBone {boneName=dyn_cloth_back_00_r}
      boxExtents=<4 20 20>
      constraintSetup {angularLimitsMax=<5 20 40>, angularLimitsMin=<-5 -20 -40>, angularTarget=<0 0 0>}
      localJointOffset=<1 0 0>
     }
     {
      boundBone {boneName=dyn_cloth_back_01_r}
      boxExtents=<4 20 20>
      constraintSetup {angularLimitsMax=<5 20 40>, angularLimitsMin=<-5 -20 -40>, angularTarget=<0 0 0>}
      localJointOffset=<1 0 0>
     }
     {
      boundBone {boneName=dyn_cloth_back_02_r}
      boxExtents=<4 20 20>
      constraintSetup {angularLimitsMax=<5 20 40>, angularLimitsMin=<-5 -20 -40>, angularTarget=<0 0 0>}
      localJointOffset=<1 0 0>
     }
     {
      boundBone {boneName=dyn_cloth_back_03_r}
      boxExtents=<4 20 20>
      constraintSetup {angularLimitsMax=<5 20 40>, angularLimitsMin=<-5 -20 -40>, angularTarget=<0 0 0>}
      localJointOffset=<1 0 0>
     }
     {
      boundBone {boneName=dyn_cloth_back_04_r}
      boxExtents=<4 20 20>
      constraintSetup {angularLimitsMax=<5 20 40>, angularLimitsMin=<-5 -20 -40>, angularTarget=<0 0 0>}
      localJointOffset=<1 0 0>
     }
    ]
    planarLimits [
     {drivingBone {boneName=thigh_r}, planeTransform {location=<0 9 0>, rotation=<rot 0 -9.9999 90>}}
     {drivingBone {boneName=calf_r}, planeTransform {location=<10 8 0>, rotation=<rot 0 0 90>}}
     {drivingBone {boneName=root}, planeTransform {location=<0 0 13>, rotation=<rot 0 0 0>, scale=<5 5 5>}}
     {drivingBone {boneName=pelvis}, planeTransform {location=<-30 0 0>, rotation=<rot 0 0 0>}}
    ]
    relativeSpaceBone {boneName=root}
   }
   n20 = "Anim Dynamics - Chain: dyn_cloth_back_00_l -> dyn_cloth_back_04_l"!(ComponentPose=@n19, Alpha=@n18, GravityOverride:Vector="0,0,300", ExternalForce="0,0,0") -> (Pose) {
    angularDampingOverride=1
    angularSpringConstant=10000
    bAngularSpring=true
    bChain=true
    bOverrideAngularDamping=true
    bUseGravityOverride=true
    bUseSphericalLimits=true
    simulationSpace=BoneRelative
    alphaBoolBlend {blendOption=Linear}
    alphaScaleBiasClamp {bInterpResult=true}
    boundBone {boneName=dyn_cloth_back_00_l}
    chainEnd {boneName=dyn_cloth_back_04_l}
    componentAppliedLinearAccClamp=<1000 1000 1000>
    componentLinearAccScale=<0.0001 0.0001 0.0001>
    componentLinearVelScale=<0.0001 0.0001 0.0001>
    gravityOverride {z=-700}
    physicsBodyDefinitions [
     {
      boundBone {boneName=dyn_cloth_back_00_l}
      boxExtents=<4 20 20>
      constraintSetup {angularLimitsMax=<5 20 40>, angularLimitsMin=<-5 -20 -40>, angularTarget=<0 0 0>}
      localJointOffset=<1 0 0>
     }
     {
      boundBone {boneName=dyn_cloth_back_01_l}
      boxExtents=<4 20 20>
      constraintSetup {angularLimitsMax=<5 20 40>, angularLimitsMin=<-5 -20 -40>, angularTarget=<0 0 0>}
      localJointOffset=<1 0 0>
     }
     {
      boundBone {boneName=dyn_cloth_back_02_l}
      boxExtents=<4 20 20>
      constraintSetup {angularLimitsMax=<5 20 40>, angularLimitsMin=<-5 -20 -40>, angularTarget=<0 0 0>}
      localJointOffset=<1 0 0>
     }
     {
      boundBone {boneName=dyn_cloth_back_03_l}
      boxExtents=<4 20 20>
      constraintSetup {angularLimitsMax=<5 20 40>, angularLimitsMin=<-5 -20 -40>, angularTarget=<0 0 0>}
      localJointOffset=<1 0 0>
     }
     {
      boundBone {boneName=dyn_cloth_back_04_l}
      boxExtents=<4 20 20>
      constraintSetup {angularLimitsMax=<5 20 40>, angularLimitsMin=<-5 -20 -40>, angularTarget=<0 0 0>}
      localJointOffset=<1 0 0>
     }
    ]
    planarLimits [
     {drivingBone {boneName=thigh_l}, planeTransform {location=<0 -9 0>, rotation=<rot 0 -9.9999 -90>}}
     {drivingBone {boneName=calf_l}, planeTransform {location=<-10 -8 0>, rotation=<rot 0 0 -90>}}
     {drivingBone {boneName=root}, planeTransform {location=<0 0 13>, rotation=<rot 0 0 0>, scale=<5 5 5>}}
     {drivingBone {boneName=pelvis}, planeTransform {location=<-30 0 0>, rotation=<rot 0 0 180>}}
    ]
    relativeSpaceBone {boneName=root}
   }
   n21 = "Anim Dynamics - Chain: dyn_cloth_front_00 -> dyn_cloth_front_04"(ComponentPose=@n20, Alpha=@n18, GravityOverride="0,0,300", ExternalForce="0,0,0") {
    angularDampingOverride=1
    angularSpringConstant=3000
    bAngularSpring=true
    bChain=true
    bOverrideAngularDamping=true
    bUseGravityOverride=true
    simulationSpace=BoneRelative
    windScale=-10
    alphaBoolBlend {blendOption=Linear}
    alphaScaleBiasClamp {bInterpResult=true}
    boundBone {boneName=dyn_cloth_front_00}
    chainEnd {boneName=dyn_cloth_front_04}
    componentAppliedLinearAccClamp=<150 150 180>
    componentLinearAccScale=<0.1 0.1 0.2>
    componentLinearVelScale=<0.3 0.3 0.4>
    gravityOverride {z=-700}
    physicsBodyDefinitions [
     {
      collisionType=CustomSphere
      sphereCollisionRadius=2
      boundBone {boneName=dyn_cloth_front_00}
      boxExtents=<4 20 20>
      constraintSetup {angularLimitsMax=<5 20 40>, angularLimitsMin=<-5 -20 -40>, angularTarget=<0 0 0>}
      localJointOffset=<1 0 0>
     }
     {
      collisionType=CustomSphere
      sphereCollisionRadius=2
      boundBone {boneName=dyn_cloth_front_01}
      boxExtents=<4 20 20>
      constraintSetup {angularLimitsMax=<5 20 40>, angularLimitsMin=<-5 -20 -40>, angularTarget=<0 0 0>}
      localJointOffset=<1 0 0>
     }
     {
      collisionType=CustomSphere
      sphereCollisionRadius=3
      boundBone {boneName=dyn_cloth_front_02}
      boxExtents=<4 20 20>
      constraintSetup {angularLimitsMax=<5 20 40>, angularLimitsMin=<-5 -20 -40>, angularTarget=<0 0 0>}
      localJointOffset=<1 0 0>
     }
     {
      collisionType=CustomSphere
      sphereCollisionRadius=3
      boundBone {boneName=dyn_cloth_front_03}
      boxExtents=<4 20 20>
      constraintSetup {angularLimitsMax=<5 20 40>, angularLimitsMin=<-5 -20 -40>, angularTarget=<0 0 0>}
      localJointOffset=<1 0 0>
     }
     {
      collisionType=CustomSphere
      sphereCollisionRadius=3
      boundBone {boneName=dyn_cloth_front_04}
      boxExtents=<4 20 20>
      constraintSetup {angularLimitsMax=<5 20 40>, angularLimitsMin=<-5 -20 -40>, angularTarget=<0 0 0>}
      localJointOffset=<1 0 0>
     }
    ]
    planarLimits [
     {drivingBone {boneName=thigh_l}, planeTransform {location=<0 0 6>, rotation=<rot -7.0001 0 0>}}
     {drivingBone {boneName=thigh_r}, planeTransform {location=<-0.1809 0 -7.5208>, rotation=<rot 6.1911 179.9997 179.9999>}}
     {drivingBone {boneName=pelvis}, planeTransform {location=<-4.5431 6.829 0>, rotation=<rot 0 1.861 90>, scale=<0.5 0.5 0.5>}}
     {drivingBone {boneName=root}, planeTransform {location=<0 0 0>, rotation=<rot 0 0 0>}}
    ]
    relativeSpaceBone {boneName=root}
   }
   n22 = Get On Ground Anim Dynamics Alpha() -> (On Ground Anim Dynamics Alpha:double)
   n23 = Anim Dynamics - Bone: dyn_cloth_back_00_r(ComponentPose=@n21, Alpha=@n22, ExternalForce="0,0,0") {
    angularDampingOverride=0.9999
    angularSpringConstant=400
    bAngularSpring=true
    bOverrideAngularDamping=true
    simulationSpace=World
    alphaBoolBlend {blendOption=Linear}
    alphaScaleBiasClamp {bInterpResult=true}
    boundBone {boneName=dyn_cloth_back_00_r}
    componentAppliedLinearAccClamp=<150 150 180>
    componentLinearAccScale=<0.1 0.1 0.2>
    componentLinearVelScale=<0.3 0.3 0.4>
    physicsBodyDefinitions [{
      sphereCollisionRadius=4
      boundBone {boneName=dyn_cloth_back_00_r}
      boxExtents=<45 35 40>
      constraintSetup {angularLimitsMax=<0 10 30>, angularLimitsMin=<0 -10 -10>, angularTarget=<0 0 0>}
      localJointOffset=<3 0 0>
     }]
    relativeSpaceBone {boneName=pelvis}
   }
   n24 = Anim Dynamics - Bone: dyn_cloth_back_01_r(ComponentPose=@n23, Alpha=@n22, ExternalForce="0,0,0") {
    angularDampingOverride=0.9999
    angularSpringConstant=400
    bAngularSpring=true
    bOverrideAngularDamping=true
    simulationSpace=World
    alphaBoolBlend {blendOption=Linear}
    alphaScaleBiasClamp {bInterpResult=true}
    boundBone {boneName=dyn_cloth_back_01_r}
    componentAppliedLinearAccClamp=<150 150 180>
    componentLinearAccScale=<0.1 0.1 0.2>
    componentLinearVelScale=<0.3 0.3 0.4>
    physicsBodyDefinitions [{
      sphereCollisionRadius=4
      boundBone {boneName=dyn_cloth_back_01_r}
      boxExtents=<30 20 25>
      constraintSetup {angularLimitsMax=<5 10 40>, angularLimitsMin=<-5 -10 -10>, angularTarget=<0 0 0>}
      localJointOffset=<3 0 -0.2>
     }]
    planarLimits [{drivingBone {boneName=thigh_r}, planeTransform {location=<30 9 0>, rotation=<rot 0 0 80>}}]
    relativeSpaceBone {boneName=pelvis}
   }
   n25 = Get On Ground Chain End Element Anim Dynamics Alpha() -> (On Ground Chain End Element Anim Dynamics Alpha:double)
   n26 = Anim Dynamics - Bone: dyn_cloth_back_02_r(ComponentPose=@n24, Alpha=@n25, ExternalForce="0,0,0") {
    angularDampingOverride=0.9999
    angularSpringConstant=100
    bAngularSpring=true
    bOverrideAngularDamping=true
    simulationSpace=World
    alphaBoolBlend {blendOption=Linear}
    alphaScaleBiasClamp {bInterpResult=true}
    boundBone {boneName=dyn_cloth_back_02_r}
    componentAppliedLinearAccClamp=<150 150 180>
    componentLinearAccScale=<0.1 0.1 0.2>
    componentLinearVelScale=<0.3 0.3 0.4>
    physicsBodyDefinitions [{
      sphereCollisionRadius=3
      boundBone {boneName=dyn_cloth_back_02_r}
      boxExtents=<30 20 25>
      constraintSetup {angularLimitsMax=<5 1 30>, angularLimitsMin=<-5 -1 -30>, angularTarget=<0 0 0>}
      localJointOffset=<2 0 -0.5>
     }]
    planarLimits [
     {drivingBone {boneName=thigh_r}, planeTransform {location=<30 9 0>, rotation=<rot 0 0 80>}}
     {drivingBone {boneName=root}, planeTransform {location=<0 0 25>, rotation=<rot 0 20 -20>}}
    ]
    relativeSpaceBone {boneName=pelvis}
   }
   n27 = Anim Dynamics - Bone: dyn_cloth_back_03_r(ComponentPose=@n26, Alpha=@n25, ExternalForce="0,0,0") {
    angularDampingOverride=0.9999
    angularSpringConstant=400
    bAngularSpring=true
    bOverrideAngularDamping=true
    bUseSphericalLimits=true
    simulationSpace=World
    alphaBoolBlend {blendOption=Linear}
    alphaScaleBiasClamp {bInterpResult=true}
    boundBone {boneName=dyn_cloth_back_03_r}
    componentAppliedLinearAccClamp=<150 150 180>
    componentLinearAccScale=<0.1 0.1 0.2>
    componentLinearVelScale=<0.3 0.3 0.4>
    physicsBodyDefinitions [{
      collisionType=CustomSphere
      sphereCollisionRadius=2
      boundBone {boneName=dyn_cloth_back_03_r}
      boxExtents=<30 20 25>
      constraintSetup {angularLimitsMax=<5 1 40>, angularLimitsMin=<-5 -1 -20>, angularTarget=<0 0 0>}
      localJointOffset=<10 0 0>
     }]
    planarLimits [{drivingBone {boneName=root}, planeTransform {location=<0 0 2>, rotation=<rot 0 0 0>}}]
    relativeSpaceBone {boneName=pelvis}
    sphericalLimits [{limitRadius=8, drivingBone {boneName=calf_r}, sphereLocalOffset=<12 0 0>}]
   }
   n28 = Anim Dynamics - Bone: dyn_cloth_back_04_r(ComponentPose=@n27, Alpha=@n25, ExternalForce="0,0,0") {
    angularDampingOverride=0.9999
    angularSpringConstant=400
    bAngularSpring=true
    bOverrideAngularDamping=true
    bUseSphericalLimits=true
    simulationSpace=World
    alphaBoolBlend {blendOption=Linear}
    alphaScaleBiasClamp {bInterpResult=true}
    boundBone {boneName=dyn_cloth_back_04_r}
    componentAppliedLinearAccClamp=<150 150 180>
    componentLinearAccScale=<0.1 0.1 0.2>
    componentLinearVelScale=<0.3 0.3 0.4>
    physicsBodyDefinitions [{
      collisionType=CustomSphere
      sphereCollisionRadius=2
      boundBone {boneName=dyn_cloth_back_04_r}
      boxExtents=<30 20 25>
      constraintSetup {angularLimitsMax=<5 2 30>, angularLimitsMin=<-5 -2 -10>, angularTarget=<0 0 0>}
      localJointOffset=<5 0 0>
     }]
    planarLimits [{drivingBone {boneName=root}, planeTransform {location=<0 0 1>, rotation=<rot 0 0 0>}}]
    relativeSpaceBone {boneName=pelvis}
    sphericalLimits [{limitRadius=10, drivingBone {boneName=calf_r}, sphereLocalOffset=<16 0 0>}]
   }
   n29 = Anim Dynamics - Bone: dyn_cloth_back_00_l(ComponentPose=@n28, Alpha=@n22, ExternalForce="0,0,0") {
    angularDampingOverride=0.9999
    angularSpringConstant=400
    bAngularSpring=true
    bOverrideAngularDamping=true
    simulationSpace=World
    alphaBoolBlend {blendOption=Linear}
    alphaScaleBiasClamp {bInterpResult=true}
    boundBone {boneName=dyn_cloth_back_00_l}
    componentAppliedLinearAccClamp=<150 150 180>
    componentLinearAccScale=<0.1 0.1 0.2>
    componentLinearVelScale=<0.3 0.3 0.4>
    physicsBodyDefinitions [{
      sphereCollisionRadius=4
      boundBone {boneName=dyn_cloth_back_00_l}
      boxExtents=<45 35 40>
      constraintSetup {angularLimitsMax=<0 10 30>, angularLimitsMin=<0 -10 -10>, angularTarget=<0 0 0>}
      localJointOffset=<-3 0 0>
     }]
    relativeSpaceBone {boneName=pelvis}
   }
   n30 = Anim Dynamics - Bone: dyn_cloth_back_01_l(ComponentPose=@n29, Alpha=@n22, ExternalForce="0,0,0") {
    angularDampingOverride=0.9999
    angularSpringConstant=400
    bAngularSpring=true
    bOverrideAngularDamping=true
    simulationSpace=World
    alphaBoolBlend {blendOption=Linear}
    alphaScaleBiasClamp {bInterpResult=true}
    boundBone {boneName=dyn_cloth_back_01_l}
    componentAppliedLinearAccClamp=<150 150 180>
    componentLinearAccScale=<0.1 0.1 0.2>
    componentLinearVelScale=<0.3 0.3 0.4>
    physicsBodyDefinitions [{
      sphereCollisionRadius=4
      boundBone {boneName=dyn_cloth_back_01_l}
      boxExtents=<30 20 25>
      constraintSetup {angularLimitsMax=<5 10 40>, angularLimitsMin=<-5 -10 -10>, angularTarget=<0 0 0>}
      localJointOffset=<-3 0 0.2>
     }]
    planarLimits [{drivingBone {boneName=thigh_l}, planeTransform {location=<-30 -9 0>, rotation=<rot 0 0 -110>}}]
    relativeSpaceBone {boneName=pelvis}
   }
   n31 = Anim Dynamics - Bone: dyn_cloth_back_02_l(ComponentPose=@n30, Alpha=@n25, ExternalForce="0,0,0") {
    angularDampingOverride=0.9999
    angularSpringConstant=400
    bAngularSpring=true
    bOverrideAngularDamping=true
    simulationSpace=World
    alphaBoolBlend {blendOption=Linear}
    alphaScaleBiasClamp {bInterpResult=true}
    boundBone {boneName=dyn_cloth_back_02_l}
    componentAppliedLinearAccClamp=<150 150 180>
    componentLinearAccScale=<0.1 0.1 0.2>
    componentLinearVelScale=<0.3 0.3 0.4>
    physicsBodyDefinitions [{
      sphereCollisionRadius=3
      boundBone {boneName=dyn_cloth_back_02_l}
      boxExtents=<30 20 25>
      constraintSetup {angularLimitsMax=<7 1 30>, angularLimitsMin=<-7 -1 -40>, angularTarget=<0 0 0>}
      localJointOffset=<-2 0 0.5>
     }]
    planarLimits [
     {drivingBone {boneName=thigh_l}, planeTransform {location=<-30 -9 0>, rotation=<rot 0 0 -110>}}
     {drivingBone {boneName=root}, planeTransform {location=<0 0 25>, rotation=<rot 0 20 -20>}}
    ]
    relativeSpaceBone {boneName=pelvis}
   }
   n32 = Anim Dynamics - Bone: dyn_cloth_back_03_l(ComponentPose=@n31, Alpha=@n25, ExternalForce="0,0,0") {
    angularDampingOverride=0.9999
    angularSpringConstant=400
    bAngularSpring=true
    bOverrideAngularDamping=true
    bUseSphericalLimits=true
    simulationSpace=World
    alphaBoolBlend {blendOption=Linear}
    alphaScaleBiasClamp {bInterpResult=true}
    boundBone {boneName=dyn_cloth_back_03_l}
    componentAppliedLinearAccClamp=<150 150 180>
    componentLinearAccScale=<0.1 0.1 0.2>
    componentLinearVelScale=<0.3 0.3 0.4>
    physicsBodyDefinitions [{
      collisionType=CustomSphere
      sphereCollisionRadius=2
      boundBone {boneName=dyn_cloth_back_03_l}
      boxExtents=<30 20 25>
      constraintSetup {angularLimitsMax=<9 25 40>, angularLimitsMin=<-9 -25 -10>, angularTarget=<0 0 0>}
      localJointOffset=<-10 0 0.5>
     }]
    planarLimits [{drivingBone {boneName=root}, planeTransform {location=<0 0 2>, rotation=<rot 0 0 0>}}]
    relativeSpaceBone {boneName=pelvis}
    sphericalLimits [{limitRadius=8, drivingBone {boneName=calf_l}, sphereLocalOffset=<-12 0 0>}]
   }
   n33 = Anim Dynamics - Bone: dyn_cloth_back_04_l(ComponentPose=@n32, Alpha=@n25, ExternalForce="0,0,0") {
    angularDampingOverride=0.9999
    angularSpringConstant=400
    bAngularSpring=true
    bOverrideAngularDamping=true
    bUseSphericalLimits=true
    simulationSpace=World
    alphaBoolBlend {blendOption=Linear}
    alphaScaleBiasClamp {bInterpResult=true}
    boundBone {boneName=dyn_cloth_back_04_l}
    componentAppliedLinearAccClamp=<150 150 180>
    componentLinearAccScale=<0.1 0.1 0.2>
    componentLinearVelScale=<0.3 0.3 0.4>
    physicsBodyDefinitions [{
      collisionType=CustomSphere
      sphereCollisionRadius=2
      boundBone {boneName=dyn_cloth_back_04_l}
      boxExtents=<30 20 25>
      constraintSetup {angularLimitsMax=<12 30 40>, angularLimitsMin=<-12 -30 -40>, angularTarget=<0 0 0>}
      localJointOffset=<-5 0 0>
     }]
    planarLimits [{drivingBone {boneName=root}, planeTransform {location=<0 0 1>, rotation=<rot 0 0 0>}}]
    relativeSpaceBone {boneName=pelvis}
    sphericalLimits [{limitRadius=10, drivingBone {boneName=calf_l}, sphereLocalOffset=<-18 0 0>}]
   }
   n34 = Anim Dynamics - Bone: dyn_cloth_front_00(ComponentPose=@n33, Alpha=@n22, ExternalForce="0,0,0") {
    angularDampingOverride=1
    angularSpringConstant=400
    bAngularSpring=true
    bOverrideAngularDamping=true
    simulationSpace=World
    alphaBoolBlend {blendOption=Linear}
    alphaScaleBiasClamp {bInterpResult=true}
    boundBone {boneName=dyn_cloth_front_00}
    componentAppliedLinearAccClamp=<150 150 180>
    componentLinearAccScale=<0.1 0.1 0.2>
    componentLinearVelScale=<0.3 0.3 0.4>
    physicsBodyDefinitions [{
      sphereCollisionRadius=4
      boundBone {boneName=dyn_cloth_front_00}
      boxExtents=<25 12 20>
      constraintSetup {angularLimitsMax=<0 0 10>, angularLimitsMin=<0 0 -30>, angularTarget=<0 0 0>}
      localJointOffset=<3 0 0>
     }]
    relativeSpaceBone {boneName=pelvis}
   }
   n35 = Anim Dynamics - Bone: dyn_cloth_front_01(ComponentPose=@n34, Alpha=@n22, ExternalForce="0,0,0") {
    angularDampingOverride=1
    angularSpringConstant=400
    bAngularSpring=true
    bOverrideAngularDamping=true
    simulationSpace=World
    alphaBoolBlend {blendOption=Linear}
    alphaScaleBiasClamp {bInterpResult=true}
    boundBone {boneName=dyn_cloth_front_01}
    componentAppliedLinearAccClamp=<150 150 180>
    componentLinearAccScale=<0.1 0.1 0.2>
    componentLinearVelScale=<0.3 0.3 0.4>
    physicsBodyDefinitions [{
      collisionType=CustomSphere
      sphereCollisionRadius=3
      boundBone {boneName=dyn_cloth_front_01}
      boxExtents=<25 12 20>
      constraintSetup {angularLimitsMax=<5 0 30>, angularLimitsMin=<-5 0 -30>, angularTarget=<0 0 0>}
      localJointOffset=<3 0 0>
     }]
    relativeSpaceBone {boneName=pelvis}
   }
   n36 = Anim Dynamics - Bone: dyn_cloth_front_02(ComponentPose=@n35, Alpha=@n22, ExternalForce="0,0,0") {
    angularDampingOverride=1
    angularSpringConstant=100
    bAngularSpring=true
    bOverrideAngularDamping=true
    simulationSpace=World
    alphaBoolBlend {blendOption=Linear}
    alphaScaleBiasClamp {bInterpResult=true}
    boundBone {boneName=dyn_cloth_front_02}
    componentAppliedLinearAccClamp=<150 150 180>
    componentLinearAccScale=<0.1 0.1 0.2>
    componentLinearVelScale=<0.3 0.3 0.4>
    physicsBodyDefinitions [{
      collisionType=CustomSphere
      sphereCollisionRadius=3
      boundBone {boneName=dyn_cloth_front_02}
      boxExtents=<25 15 20>
      constraintSetup {angularLimitsMax=<7 0 30>, angularLimitsMin=<-7 0 -30>, angularTarget=<0 0 0>}
      localJointOffset=<2 0 0>
     }]
    relativeSpaceBone {boneName=pelvis}
   }
   n37 = Anim Dynamics - Bone: dyn_cloth_front_03(ComponentPose=@n36, Alpha=@n22, ExternalForce="0,0,0") {
    angularDampingOverride=1
    angularSpringConstant=400
    bAngularSpring=true
    bOverrideAngularDamping=true
    bUseSphericalLimits=true
    simulationSpace=World
    alphaBoolBlend {blendOption=Linear}
    alphaScaleBiasClamp {bInterpResult=true}
    boundBone {boneName=dyn_cloth_front_03}
    componentAppliedLinearAccClamp=<150 150 180>
    componentLinearAccScale=<0.1 0.1 0.2>
    componentLinearVelScale=<0.3 0.3 0.4>
    physicsBodyDefinitions [{
      collisionType=CustomSphere
      sphereCollisionRadius=3
      boundBone {boneName=dyn_cloth_front_03}
      boxExtents=<25 15 20>
      constraintSetup {angularLimitsMax=<9 15 40>, angularLimitsMin=<-9 -15 -40>, angularTarget=<0 0 0>}
      localJointOffset=<2 0 0>
     }]
    relativeSpaceBone {boneName=pelvis}
   }
   n38 = Anim Dynamics - Bone: dyn_cloth_front_04(ComponentPose=@n37, Alpha=@n22, ExternalForce="0,0,0") {
    angularDampingOverride=1
    angularSpringConstant=400
    bAngularSpring=true
    bOverrideAngularDamping=true
    bUseSphericalLimits=true
    simulationSpace=World
    alphaBoolBlend {blendOption=Linear}
    alphaScaleBiasClamp {bInterpResult=true}
    boundBone {boneName=dyn_cloth_front_04}
    componentAppliedLinearAccClamp=<150 150 180>
    componentLinearAccScale=<0.1 0.1 0.2>
    componentLinearVelScale=<0.3 0.3 0.4>
    physicsBodyDefinitions [{
      collisionType=CustomSphere
      sphereCollisionRadius=3
      boundBone {boneName=dyn_cloth_front_04}
      boxExtents=<25 15 20>
      constraintSetup {angularLimitsMax=<9 15 40>, angularLimitsMin=<-9 -15 -40>, angularTarget=<0 0 0>}
      localJointOffset=<2 0 0>
     }]
    relativeSpaceBone {boneName=pelvis}
   }
   n39 = Get bIsPlayingMeleeAnim()
   n40 = "Transform (Modify) Bone - Bone: dyn_cloth_front_03"!(ComponentPose=@n38, bAlphaBoolEnabled:bool=@n39, Translation="0.728,0,0", Rotation="0.821,25,-0.945", Scale="1,1,1") -> (Pose) {
    alphaInputType=Bool
    rotationMode=BMM_Additive
    rotationSpace=BCS_BoneSpace
    translationMode=BMM_Additive
    translationSpace=BCS_BoneSpace
    alphaBoolBlend {blendInTime=0.3, blendOption=Linear, blendOutTime=0.8}
    boneToModify {boneName=dyn_cloth_front_03}
    rotation=<rot 0.821 25 -0.945>
    translation {x=0.728}
   }
   n41 = "Transform (Modify) Bone - Bone: dyn_cloth_front_MSTR"(ComponentPose=@n40, bAlphaBoolEnabled=@n39, Translation="-0.398,-5.019,0.362", Rotation="3.690,-7.486,8.529", Scale="1,1,1") {
    alphaInputType=Bool
    rotationMode=BMM_Additive
    rotationSpace=BCS_BoneSpace
    translationMode=BMM_Additive
    translationSpace=BCS_BoneSpace
    alphaBoolBlend {blendInTime=0.3, blendOption=Linear, blendOutTime=0.8}
    boneToModify {boneName=dyn_cloth_front_MSTR}
    rotation=<rot 3.6899 -7.4861 8.529>
    translation=<-0.3981 -5.0194 0.3621>
   }
   n42 = Get bIsTacticalSprinting()
   n43 = "Transform (Modify) Bone - Bone: dyn_cloth_back_00_r"!(ComponentPose=@n41, bAlphaBoolEnabled:bool=@n42, Translation="0,0,0", Rotation="0,20,0", Scale="1,1,1") -> (Pose) {
    alphaInputType=Bool
    rotationMode=BMM_Additive
    rotationSpace=BCS_BoneSpace
    translationMode=BMM_Additive
    translationSpace=BCS_BoneSpace
    alphaBoolBlend {blendInTime=0.3, blendOption=ExpInOut, blendOutTime=0.3}
    boneToModify {boneName=dyn_cloth_back_00_r}
    rotation {yaw=20}
   }
   n44 = "Transform (Modify) Bone - Bone: dyn_cloth_back_00_l"!(ComponentPose=@n43, bAlphaBoolEnabled:bool=@n42, Translation="0,0,0", Rotation="0,20,0", Scale="1,1,1") -> (Pose) {
    alphaInputType=Bool
    rotationMode=BMM_Additive
    rotationSpace=BCS_BoneSpace
    translationMode=BMM_Additive
    translationSpace=BCS_BoneSpace
    alphaBoolBlend {blendInTime=0.3, blendOption=ExpInOut, blendOutTime=0.3}
    boneToModify {boneName=dyn_cloth_back_00_l}
    rotation {yaw=20}
   }
   n45 = "Transform (Modify) Bone - Bone: dyn_cloth_back_01_r"!(ComponentPose=@n44, bAlphaBoolEnabled:bool=@n17, Translation="0,0,0", Rotation="0,20,0", Scale="1,1,1") -> (Pose) {
    alphaInputType=Bool
    rotationMode=BMM_Additive
    rotationSpace=BCS_BoneSpace
    translationMode=BMM_Additive
    translationSpace=BCS_BoneSpace
    alphaBoolBlend {blendInTime=0.3, blendOption=ExpInOut, blendOutTime=0.3}
    boneToModify {boneName=dyn_cloth_back_01_r}
    rotation {yaw=20}
   }
   n46 = "Transform (Modify) Bone - Bone: dyn_cloth_back_01_l"!(ComponentPose=@n45, bAlphaBoolEnabled:bool=@n17, Translation="0,0,0", Rotation="0,20,0", Scale="1,1,1") -> (Pose) {
    alphaInputType=Bool
    rotationMode=BMM_Additive
    rotationSpace=BCS_BoneSpace
    translationMode=BMM_Additive
    translationSpace=BCS_BoneSpace
    alphaBoolBlend {blendInTime=0.3, blendOption=ExpInOut, blendOutTime=0.3}
    boneToModify {boneName=dyn_cloth_back_01_l}
    rotation {yaw=20}
   }
   n47 = "Transform (Modify) Bone - Bone: dyn_cloth_back_02_r"(ComponentPose=@n46, bAlphaBoolEnabled=@n17, Translation="0,0,0", Rotation="0,20,0", Scale="1,1,1") {
    alphaInputType=Bool
    rotationMode=BMM_Additive
    rotationSpace=BCS_BoneSpace
    translationMode=BMM_Additive
    translationSpace=BCS_BoneSpace
    alphaBoolBlend {blendInTime=0.3, blendOption=ExpInOut, blendOutTime=0.3}
    boneToModify {boneName=dyn_cloth_back_02_r}
    rotation {yaw=20}
   }
   n48 = "Transform (Modify) Bone - Bone: dyn_cloth_back_02_l"(ComponentPose=@n47, bAlphaBoolEnabled=@n17, Translation="0,0,0", Rotation="0,20,0", Scale="1,1,1") {
    alphaInputType=Bool
    rotationMode=BMM_Additive
    rotationSpace=BCS_BoneSpace
    translationMode=BMM_Additive
    translationSpace=BCS_BoneSpace
    alphaBoolBlend {blendInTime=0.3, blendOption=ExpInOut, blendOutTime=0.3}
    boneToModify {boneName=dyn_cloth_back_02_l}
    rotation {yaw=20}
   }
   n49 = "Transform (Modify) Bone - Bone: dyn_cloth_back_03_r"(ComponentPose=@n48, bAlphaBoolEnabled=@n17, Translation="0,0,0", Rotation="0,30,0", Scale="1,1,1") {
    alphaInputType=Bool
    rotationMode=BMM_Additive
    rotationSpace=BCS_BoneSpace
    translationMode=BMM_Additive
    translationSpace=BCS_BoneSpace
    alphaBoolBlend {blendInTime=0.3, blendOption=ExpInOut, blendOutTime=0.3}
    boneToModify {boneName=dyn_cloth_back_03_r}
    rotation {yaw=30}
   }
   n50 = "Transform (Modify) Bone - Bone: dyn_cloth_back_03_l"(ComponentPose=@n49, bAlphaBoolEnabled=@n17, Translation="0,0,0", Rotation="0,30,0", Scale="1,1,1") {
    alphaInputType=Bool
    rotationMode=BMM_Additive
    rotationSpace=BCS_BoneSpace
    translationMode=BMM_Additive
    translationSpace=BCS_BoneSpace
    alphaBoolBlend {blendInTime=0.3, blendOption=ExpInOut, blendOutTime=0.3}
    boneToModify {boneName=dyn_cloth_back_03_l}
    rotation {yaw=30}
   }
   n51 = "Transform (Modify) Bone - Bone: dyn_cloth_back_03_r"!(ComponentPose=@n50, bAlphaBoolEnabled=@n2) -> (Pose) {
    alphaInputType=Bool
    rotationMode=BMM_Additive
    rotationSpace=BCS_BoneSpace
    translationSpace=BCS_ParentBoneSpace
    alphaBoolBlend {blendInTime=0.3, blendOption=ExpInOut, blendOutTime=0.3}
    boneToModify {boneName=dyn_cloth_back_03_r}
    rotation {yaw=30}
   }
   n52 = "Transform (Modify) Bone - Bone: dyn_cloth_back_03_l"!(ComponentPose=@n51, bAlphaBoolEnabled=@n2) -> (Pose) {
    alphaInputType=Bool
    rotationMode=BMM_Additive
    rotationSpace=BCS_BoneSpace
    translationSpace=BCS_ParentBoneSpace
    alphaBoolBlend {blendInTime=0.3, blendOption=ExpInOut, blendOutTime=0.3}
    boneToModify {boneName=dyn_cloth_back_03_l}
    rotation {yaw=30}
   }
   n53 = "Transform (Modify) Bone - Bone: dyn_cloth_back_04_r"!(ComponentPose=@n52, bAlphaBoolEnabled:bool=@n2) -> (Pose) {
    alphaInputType=Bool
    rotationMode=BMM_Additive
    rotationSpace=BCS_BoneSpace
    translationSpace=BCS_ParentBoneSpace
    alphaBoolBlend {blendInTime=0.3, blendOption=ExpInOut, blendOutTime=0.3}
    boneToModify {boneName=dyn_cloth_back_04_r}
    rotation {yaw=-15}
   }
   n54 = "Transform (Modify) Bone - Bone: dyn_cloth_back_04_l"!(ComponentPose=@n53, bAlphaBoolEnabled:bool=@n2) -> (Pose) {
    alphaInputType=Bool
    rotationMode=BMM_Additive
    rotationSpace=BCS_BoneSpace
    translationSpace=BCS_ParentBoneSpace
    alphaBoolBlend {blendInTime=0.3, blendOption=ExpInOut, blendOutTime=0.3}
    boneToModify {boneName=dyn_cloth_back_04_l}
    rotation {yaw=-15}
   }
   n55 = "Transform (Modify) Bone - Bone: dyn_cloth_front_MSTR"(ComponentPose=@n54, bAlphaBoolEnabled=@n2, Translation="0,3,3", Rotation="0,0,0", Scale="1,1,1") {
    alphaInputType=Bool
    rotationMode=BMM_Replace
    rotationSpace=BCS_BoneSpace
    translationMode=BMM_Additive
    translationSpace=BCS_ParentBoneSpace
    alphaBoolBlend {blendInTime=0.3, blendOption=ExpInOut, blendOutTime=0.3}
    boneToModify {boneName=dyn_cloth_front_MSTR}
    translation {y=3, z=3}
   }
   n56 = "Transform (Modify) Bone - Bone: dyn_cloth_back_00_r"!(ComponentPose=@n55, bAlphaBoolEnabled=@n2) -> (Pose) {
    alphaInputType=Bool
    rotationMode=BMM_Additive
    rotationSpace=BCS_BoneSpace
    translationMode=BMM_Additive
    translationSpace=BCS_ParentBoneSpace
    alphaBoolBlend {blendInTime=0.3, blendOption=ExpInOut, blendOutTime=0.3}
    boneToModify {boneName=dyn_cloth_back_00_r}
    rotation {yaw=30}
    translation=<8 5>
   }
   n57 = "Transform (Modify) Bone - Bone: dyn_cloth_back_00_l"!(ComponentPose=@n56, bAlphaBoolEnabled=@n2) -> (Pose) {
    alphaInputType=Bool
    rotationMode=BMM_Additive
    rotationSpace=BCS_BoneSpace
    translationMode=BMM_Additive
    translationSpace=BCS_ParentBoneSpace
    alphaBoolBlend {blendInTime=0.3, blendOption=ExpInOut, blendOutTime=0.3}
    boneToModify {boneName=dyn_cloth_back_00_l}
    rotation {yaw=30}
    translation=<8 5>
   }
   n58 = "Transform (Modify) Bone - Bone: dyn_cloth_front_MSTR"!(ComponentPose=@n57, bAlphaBoolEnabled=@n2) -> (Pose) {
    alphaInputType=Bool
    rotationMode=BMM_Replace
    rotationSpace=BCS_BoneSpace
    translationMode=BMM_Additive
    translationSpace=BCS_ParentBoneSpace
    alphaBoolBlend {blendInTime=0.3, blendOption=ExpInOut, blendOutTime=0.3}
    boneToModify {boneName=dyn_cloth_front_MSTR}
    translation {z=3}
   }
   n59 = "Transform (Modify) Bone - Bone: dyn_cloth_front_01"!(ComponentPose=@n58, bAlphaBoolEnabled:bool=@n2, Translation="0,-3,0", Rotation="0,0,0", Scale="1,1,1") -> (Pose) {
    alphaInputType=Bool
    rotationMode=BMM_Replace
    rotationSpace=BCS_BoneSpace
    translationMode=BMM_Replace
    translationSpace=BCS_BoneSpace
    alphaBoolBlend {blendInTime=0.3, blendOption=ExpInOut, blendOutTime=0.3}
    boneToModify {boneName=dyn_cloth_front_01}
    translation {y=-3}
   }
   n60 = "Transform (Modify) Bone - Bone: dyn_cloth_front_02"!(ComponentPose=@n59, bAlphaBoolEnabled:bool=@n2, Translation="0,0,0", Rotation="0,20,0", Scale="1,1,1") -> (Pose) {
    alphaInputType=Bool
    rotationMode=BMM_Replace
    rotationSpace=BCS_BoneSpace
    translationMode=BMM_Replace
    translationSpace=BCS_BoneSpace
    alphaBoolBlend {blendInTime=0.3, blendOption=ExpInOut, blendOutTime=0.3}
    boneToModify {boneName=dyn_cloth_front_02}
    rotation {yaw=20}
   }
   n61 = "Transform (Modify) Bone - Bone: dyn_cloth_back_01_l"(ComponentPose=@n60, bAlphaBoolEnabled=@n16, Translation="0,0,0", Rotation="1.719,-25.405,0.886", Scale="1,1,1") {
    alphaInputType=Bool
    rotationMode=BMM_Additive
    rotationSpace=BCS_BoneSpace
    translationMode=BMM_Additive
    translationSpace=BCS_BoneSpace
    alphaBoolBlend {blendInTime=0.7, blendOutTime=0.8}
    boneToModify {boneName=dyn_cloth_back_01_l}
    rotation=<rot 1.719 -25.405 0.886>
   }
   n62 = "Transform (Modify) Bone - Bone: dyn_cloth_back_01_r"(ComponentPose=@n61, bAlphaBoolEnabled=@n16, Translation="0,0,0", Rotation="-0.318,-27.358,-4.137", Scale="1,1,1") {
    alphaInputType=Bool
    rotationMode=BMM_Replace
    rotationSpace=BCS_BoneSpace
    translationMode=BMM_Replace
    translationSpace=BCS_BoneSpace
    alphaBoolBlend {blendInTime=0.7, blendOutTime=0.8}
    boneToModify {boneName=dyn_cloth_back_01_r}
    rotation=<rot -0.318 -27.358 -4.137>
   }
   n63 = "Transform (Modify) Bone - Bone: dyn_cloth_back_00_l"!(
    ComponentPose=@n62
    bAlphaBoolEnabled=@n16
    Translation:Vector="-0.371,2.840,0.615"
    Rotation:Rotator="5.703,6.238,-12.232"
    Scale:Vector="1,1,1"
   ) -> (Pose) {
    alphaInputType=Bool
    rotationMode=BMM_Additive
    rotationSpace=BCS_BoneSpace
    translationMode=BMM_Additive
    translationSpace=BCS_BoneSpace
    alphaBoolBlend {blendInTime=2, blendOutTime=2}
    boneToModify {boneName=dyn_cloth_back_00_l}
    rotation=<rot 5.703 6.238 -12.232>
    translation=<-0.371 2.84 0.615>
   }
   n64 = "Transform (Modify) Bone - Bone: dyn_cloth_back_00_r"!(
    ComponentPose=@n63
    bAlphaBoolEnabled=@n16
    Translation:Vector="0.547,-2.907,-0.490"
    Rotation:Rotator="0.448,11.257,-9.425"
    Scale:Vector="1,1,1"
   ) -> (Pose) {
    alphaInputType=Bool
    rotationMode=BMM_Replace
    rotationSpace=BCS_BoneSpace
    translationMode=BMM_Replace
    translationSpace=BCS_BoneSpace
    alphaBoolBlend {blendInTime=2, blendOutTime=2}
    boneToModify {boneName=dyn_cloth_back_00_r}
    rotation=<rot 0.448 11.257 -9.425>
    translation=<0.547 -2.907 -0.49>
   }
   n65 = "Transform (Modify) Bone - Bone: dyn_cloth_back_02_r"(ComponentPose=@n64, bAlphaBoolEnabled=@n16, Translation="-1.044,2.946,0.000", Rotation="0.000,19.514,-0.018", Scale="1,1,1") {
    alphaInputType=Bool
    rotationMode=BMM_Additive
    rotationSpace=BCS_BoneSpace
    translationMode=BMM_Additive
    translationSpace=BCS_BoneSpace
    alphaBoolBlend {blendInTime=0.3, blendOption=ExpInOut, blendOutTime=0.3}
    boneToModify {boneName=dyn_cloth_back_02_r}
    rotation=<rot 0 19.514 -0.018>
    translation=<-1.044 2.946>
   }
   n66 = "Transform (Modify) Bone - Bone: dyn_cloth_back_02_l"(ComponentPose=@n65, bAlphaBoolEnabled=@n16, Translation="2.147,-3.805,0.000", Rotation="0.000,29.424,0.002", Scale="1,1,1") {
    alphaInputType=Bool
    rotationMode=BMM_Additive
    rotationSpace=BCS_BoneSpace
    translationMode=BMM_Additive
    translationSpace=BCS_BoneSpace
    alphaBoolBlend {blendInTime=0.3, blendOption=ExpInOut, blendOutTime=0.3}
    boneToModify {boneName=dyn_cloth_back_02_l}
    rotation=<rot 0 29.424 0.002>
    translation=<2.147 -3.805>
   }
   n67 = "Transform (Modify) Bone - Bone: dyn_cloth_back_01_r"(ComponentPose=@n66, bAlphaBoolEnabled=@n15, Translation="0,-5,0", Rotation="0.143,-39.592,-0.050", Scale="1,1,1") {
    alphaInputType=Bool
    rotationMode=BMM_Additive
    rotationSpace=BCS_BoneSpace
    translationMode=BMM_Additive
    translationSpace=BCS_BoneSpace
    alphaBoolBlend {blendInTime=0.3, blendOption=ExpInOut, blendOutTime=0.3}
    boneToModify {boneName=dyn_cloth_back_01_r}
    rotation=<rot 0.143 -39.592 -0.05>
    translation {y=-5}
   }
   n68 = "Transform (Modify) Bone - Bone: dyn_cloth_back_01_l"(ComponentPose=@n67, bAlphaBoolEnabled=@n15, Translation="0,5,0", Rotation="0.160,-38.699,-0.044", Scale="1,1,1") {
    alphaInputType=Bool
    rotationMode=BMM_Additive
    rotationSpace=BCS_BoneSpace
    translationMode=BMM_Additive
    translationSpace=BCS_BoneSpace
    alphaBoolBlend {blendInTime=0.3, blendOption=ExpInOut, blendOutTime=0.3}
    boneToModify {boneName=dyn_cloth_back_01_l}
    rotation=<rot 0.16 -38.6991 -0.044>
    translation {y=5}
   }
   n69 = "Transform (Modify) Bone - Bone: dyn_cloth_back_02_r"(ComponentPose=@n68, bAlphaBoolEnabled=@n15, Translation="0,0,0", Rotation="-0.325,5.527,0.124", Scale="1,1,1") {
    alphaInputType=Bool
    rotationMode=BMM_Additive
    rotationSpace=BCS_BoneSpace
    translationMode=BMM_Additive
    translationSpace=BCS_BoneSpace
    alphaBoolBlend {blendInTime=0.3, blendOption=ExpInOut, blendOutTime=0.3}
    boneToModify {boneName=dyn_cloth_back_02_r}
    rotation=<rot -0.325 5.527 0.124>
   }
   n70 = "Transform (Modify) Bone - Bone: dyn_cloth_back_02_l"(ComponentPose=@n69, bAlphaBoolEnabled=@n15, Translation="0,0,0", Rotation="-0.310,7.229,0.076", Scale="1,1,1") {
    alphaInputType=Bool
    rotationMode=BMM_Additive
    rotationSpace=BCS_BoneSpace
    translationMode=BMM_Additive
    translationSpace=BCS_BoneSpace
    alphaBoolBlend {blendInTime=0.3, blendOption=ExpInOut, blendOutTime=0.3}
    boneToModify {boneName=dyn_cloth_back_02_l}
    rotation=<rot -0.31 7.229 0.076>
   }
   n71 = "Transform (Modify) Bone - Bone: dyn_cloth_back_00_r"(ComponentPose=@n70, bAlphaBoolEnabled=@n15, Translation="0,0,0", Rotation="0,-10,0", Scale="1,1,1") {
    alphaInputType=Bool
    rotationMode=BMM_Additive
    rotationSpace=BCS_BoneSpace
    translationMode=BMM_Additive
    translationSpace=BCS_BoneSpace
    alphaBoolBlend {blendInTime=0.3, blendOption=ExpInOut, blendOutTime=0.3}
    boneToModify {boneName=dyn_cloth_back_00_r}
    rotation {yaw=-10}
   }
   n72 = "Transform (Modify) Bone - Bone: dyn_cloth_back_00_l"(ComponentPose=@n71, bAlphaBoolEnabled=@n15, Translation="0,0,0", Rotation="0,-10,0", Scale="1,1,1") {
    alphaInputType=Bool
    rotationMode=BMM_Additive
    rotationSpace=BCS_BoneSpace
    translationMode=BMM_Additive
    translationSpace=BCS_BoneSpace
    alphaBoolBlend {blendInTime=0.3, blendOption=ExpInOut, blendOutTime=0.3}
    boneToModify {boneName=dyn_cloth_back_00_l}
    rotation {yaw=-10}
   }
   n73 = "Transform (Modify) Bone - Bone: dyn_cloth_front_00"(ComponentPose=@n72, bAlphaBoolEnabled=@n15, Translation="0,0,0", Rotation="0,-5,0", Scale="1,1,1") {
    alphaInputType=Bool
    rotationMode=BMM_Replace
    rotationSpace=BCS_BoneSpace
    translationMode=BMM_Replace
    translationSpace=BCS_BoneSpace
    alphaBoolBlend {blendInTime=0.3, blendOption=ExpInOut, blendOutTime=0.3}
    boneToModify {boneName=dyn_cloth_front_00}
    rotation {yaw=-5}
   }
   n74 = "Transform (Modify) Bone - Bone: dyn_cloth_front_01"(ComponentPose=@n73, bAlphaBoolEnabled=@n15, Translation="0,0,0", Rotation="0,10,0", Scale="1,1,1") {
    alphaInputType=Bool
    rotationMode=BMM_Replace
    rotationSpace=BCS_BoneSpace
    translationMode=BMM_Replace
    translationSpace=BCS_BoneSpace
    alphaBoolBlend {blendInTime=0.3, blendOption=ExpInOut, blendOutTime=0.3}
    boneToModify {boneName=dyn_cloth_front_01}
    rotation {yaw=10}
   }
   n75 = "Transform (Modify) Bone - Bone: dyn_cloth_front_02"(ComponentPose=@n74, bAlphaBoolEnabled=@n15, Translation="0,0,0", Rotation="0,10,0", Scale="1,1,1") {
    alphaInputType=Bool
    rotationMode=BMM_Replace
    rotationSpace=BCS_BoneSpace
    translationMode=BMM_Replace
    translationSpace=BCS_BoneSpace
    alphaBoolBlend {blendInTime=0.3, blendOption=ExpInOut, blendOutTime=0.3}
    boneToModify {boneName=dyn_cloth_front_02}
    rotation {yaw=10}
   }
   n76 = "Transform (Modify) Bone - Bone: dyn_cloth_back_00_r"(ComponentPose=@n75, bAlphaBoolEnabled=@n14, Translation="0,0,0", Rotation="-23.163,-25.832,6.292", Scale="1,1,1") {
    alphaInputType=Bool
    rotationMode=BMM_Additive
    rotationSpace=BCS_BoneSpace
    translationMode=BMM_Additive
    translationSpace=BCS_BoneSpace
    alphaBoolBlend {blendInTime=0.3, blendOption=ExpInOut, blendOutTime=0.3}
    boneToModify {boneName=dyn_cloth_back_00_r}
    rotation=<rot -23.163 -25.832 6.292>
   }
   n77 = "Transform (Modify) Bone - Bone: dyn_cloth_back_00_l"(ComponentPose=@n76, bAlphaBoolEnabled=@n14, Translation="0,0,0", Rotation="-18.462,-38.206,3.163", Scale="1,1,1") {
    alphaInputType=Bool
    rotationMode=BMM_Additive
    rotationSpace=BCS_BoneSpace
    translationMode=BMM_Additive
    translationSpace=BCS_BoneSpace
    alphaBoolBlend {blendInTime=0.3, blendOption=ExpInOut, blendOutTime=0.3}
    boneToModify {boneName=dyn_cloth_back_00_l}
    rotation=<rot -18.462 -38.2061 3.163>
   }
   n78 = "Transform (Modify) Bone - Bone: dyn_cloth_back_01_r"(ComponentPose=@n77, bAlphaBoolEnabled=@n14, Translation="0,-5,0", Rotation="-18.936,-54.625,5.157", Scale="1,1,1") {
    alphaInputType=Bool
    rotationMode=BMM_Additive
    rotationSpace=BCS_BoneSpace
    translationMode=BMM_Additive
    translationSpace=BCS_BoneSpace
    alphaBoolBlend {blendInTime=0.3, blendOption=ExpInOut, blendOutTime=0.3}
    boneToModify {boneName=dyn_cloth_back_01_r}
    rotation=<rot -18.936 -54.625 5.157>
    translation {y=-5}
   }
   n79 = "Transform (Modify) Bone - Bone: dyn_cloth_back_01_l"(ComponentPose=@n78, bAlphaBoolEnabled=@n14, Translation="0,5,0", Rotation="-21.377,-55.927,7.638", Scale="1,1,1") {
    alphaInputType=Bool
    rotationMode=BMM_Additive
    rotationSpace=BCS_BoneSpace
    translationMode=BMM_Additive
    translationSpace=BCS_BoneSpace
    alphaBoolBlend {blendInTime=0.3, blendOption=ExpInOut, blendOutTime=0.3}
    boneToModify {boneName=dyn_cloth_back_01_l}
    rotation=<rot -21.377 -55.927 7.638>
    translation {y=5}
   }
   n80 = "Transform (Modify) Bone - Bone: dyn_cloth_back_02_r"(ComponentPose=@n79, bAlphaBoolEnabled=@n14, Translation="0,0,0", Rotation="-4.346,-40.320,1.074", Scale="1,1,1") {
    alphaInputType=Bool
    rotationMode=BMM_Additive
    rotationSpace=BCS_BoneSpace
    translationMode=BMM_Additive
    translationSpace=BCS_BoneSpace
    alphaBoolBlend {blendInTime=0.3, blendOption=ExpInOut, blendOutTime=0.3}
    boneToModify {boneName=dyn_cloth_back_02_r}
    rotation=<rot -4.346 -40.32 1.074>
   }
   n81 = "Transform (Modify) Bone - Bone: dyn_cloth_back_02_l"(ComponentPose=@n80, bAlphaBoolEnabled=@n14, Translation="0,0,0", Rotation="-6.377,-45.168,1.856", Scale="1,1,1") {
    alphaInputType=Bool
    rotationMode=BMM_Additive
    rotationSpace=BCS_BoneSpace
    translationMode=BMM_Additive
    translationSpace=BCS_BoneSpace
    alphaBoolBlend {blendInTime=0.3, blendOption=ExpInOut, blendOutTime=0.3}
    boneToModify {boneName=dyn_cloth_back_02_l}
    rotation=<rot -6.377 -45.168 1.856>
   }
   n82 = "Transform (Modify) Bone - Bone: dyn_cloth_front_00"(ComponentPose=@n81, bAlphaBoolEnabled=@n14, Translation="0,0,0", Rotation="13.446,47.383,-2.042", Scale="1,1,1") {
    alphaInputType=Bool
    rotationMode=BMM_Replace
    rotationSpace=BCS_BoneSpace
    translationMode=BMM_Replace
    translationSpace=BCS_BoneSpace
    alphaBoolBlend {blendInTime=0.3, blendOption=ExpInOut, blendOutTime=0.3}
    boneToModify {boneName=dyn_cloth_front_00}
    rotation=<rot 13.446 47.383 -2.042>
   }
   n83 = "Transform (Modify) Bone - Bone: dyn_cloth_front_01"(ComponentPose=@n82, bAlphaBoolEnabled=@n14, Translation="0,0,0", Rotation="3.346,33.365,-0.127", Scale="1,1,1") {
    alphaInputType=Bool
    rotationMode=BMM_Replace
    rotationSpace=BCS_BoneSpace
    translationMode=BMM_Replace
    translationSpace=BCS_BoneSpace
    alphaBoolBlend {blendInTime=0.3, blendOption=ExpInOut, blendOutTime=0.3}
    boneToModify {boneName=dyn_cloth_front_01}
    rotation=<rot 3.346 33.365 -0.127>
   }
   n84 = "Transform (Modify) Bone - Bone: dyn_cloth_front_02"(ComponentPose=@n83, bAlphaBoolEnabled=@n14, Translation="0,0,0", Rotation="14.172,36.438,-1.270", Scale="1,1,1") {
    alphaInputType=Bool
    rotationMode=BMM_Replace
    rotationSpace=BCS_BoneSpace
    translationMode=BMM_Replace
    translationSpace=BCS_BoneSpace
    alphaBoolBlend {blendInTime=0.3, blendOption=ExpInOut, blendOutTime=0.3}
    boneToModify {boneName=dyn_cloth_front_02}
    rotation=<rot 14.172 36.438 -1.27>
   }
   n85 = "Transform (Modify) Bone - Bone: dyn_cloth_front_03"(ComponentPose=@n84, bAlphaBoolEnabled=@n14, Translation="0,0,0", Rotation="6.998,35.009,-0.800", Scale="1,1,1") {
    alphaInputType=Bool
    rotationMode=BMM_Replace
    rotationSpace=BCS_BoneSpace
    translationMode=BMM_Replace
    translationSpace=BCS_BoneSpace
    alphaBoolBlend {blendInTime=0.3, blendOption=ExpInOut, blendOutTime=0.3}
    boneToModify {boneName=dyn_cloth_front_03}
    rotation=<rot 6.998 35.009 -0.8>
   }
   n86 = "Transform (Modify) Bone - Bone: dyn_cloth_front_MSTR"(ComponentPose=Local To Component(LocalPose=Use cached pose 'PreRigid'()), bAlphaBoolEnabled=Get bIsSliding()) {
    alphaInputType=Bool
    rotationMode=BMM_Additive
    rotationSpace=BCS_ParentBoneSpace
    translationMode=BMM_Additive
    translationSpace=BCS_BoneSpace
    alphaBoolBlend {blendInTime=0.3, blendOption=ExpInOut, blendOutTime=0.3}
    boneToModify {boneName=dyn_cloth_front_MSTR}
    rotation {pitch=-20, yaw=30}
    translation {y=28}
   }
   n87 = RigidBody!(ComponentPose=@n86, Alpha, OverrideWorldGravity:Vector=Get Gravity Override(), ExternalForce) -> (Pose) {
    bOverrideWorldGravity=true
    overridePhysicsAsset=&/F_MED_TalonHime_Physics
    simulationSpace=BaseBoneSpace
    alphaBoolBlend {blendOption=Linear}
    baseBoneRef {boneName=root}
    componentAppliedLinearAccClamp=<4000 4000 4000>
    componentLinearAccScale=<0.45 0.45 0.45>
    componentLinearVelScale=<0.45 0.45 0.45>
   }
   Output Pose(Result=Blend Poses by bool(BlendPose_0=Component To Local(ComponentPose=@n87), BlendPose_1=Component To Local(ComponentPose=@n85), bActiveValue=Get bRigidBodyEnabled()) {blendType=Linear}) {layerGroup=None}
   n88 = Use cached pose 'Shoulder pads cache'()
   Save cached pose 'PreRigid'(Pose=Component To Local(ComponentPose=Local To Component(LocalPose=@n88))) {cacheName=PreRigid}
   n89 = Control Rig(Source=Copy Pose From Mesh() {bUseAttachedParent=true}) {
    alphaBoolBlend {blendOption=Linear}
    controlRigAssetReference {blueprintRigClass=&/F_MED_TalonHime_ControlRig.F_MED_TalonHime_ControlRig_C}
   }
   n90 = "Transform (Modify) Bone - Bone: shoulder_pad_r"!(
    ComponentPose=Local To Component(LocalPose=@n89)
    Alpha:float=Get Right Pad alpha() -> (Right Pad alpha:double)
    Translation="-3.057,2.956,0.724"
    Rotation="47.472,24.738,-73.273"
    Scale="1,1,1"
   ) -> (Pose) {
    rotationMode=BMM_Additive
    rotationSpace=BCS_BoneSpace
    translationMode=BMM_Additive
    translationSpace=BCS_BoneSpace
    alphaBoolBlend {blendOption=Linear}
    alphaScaleBiasClamp {interpSpeedDecreasing=20, interpSpeedIncreasing=20}
    boneToModify {boneName=shoulder_pad_r}
    rotation=<rot 47.472 24.738 -73.273>
    translation=<-3.057 2.956 0.724>
   }
   n91 = "Constraint - shoulder_pad_r (R)"(ComponentPose=@n90, Alpha=0.3, ConstraintWeights_0=1) {
    alphaBoolBlend {blendOption=Linear}
    boneToModify {boneName=shoulder_pad_r}
    constraintSetup [{transformType=Rotation, targetBone {boneName=upperarm_twist_01_r}}]
    constraintWeights [1]
   }
   n92 = "Transform (Modify) Bone - Bone: shoulder_pad_l"(
    ComponentPose=@n91
    Alpha=Get Left Pad alpha() -> (Left Pad alpha:double)
    Translation="3.856,-1.013,-1.275"
    Rotation="41.707,30.232,-56.275"
    Scale="1,1,1"
   ) {
    rotationMode=BMM_Additive
    rotationSpace=BCS_BoneSpace
    translationMode=BMM_Additive
    translationSpace=BCS_BoneSpace
    alphaBoolBlend {blendOption=Linear}
    alphaScaleBiasClamp {interpSpeedIncreasing=5}
    boneToModify {boneName=shoulder_pad_l}
    rotation=<rot 41.707 30.232 -56.275>
    translation=<3.856 -1.013 -1.275>
   }
   n93 = "Constraint - shoulder_pad_l (R)"(ComponentPose=@n92, Alpha=0.3, ConstraintWeights_0=1) {
    alphaBoolBlend {blendOption=Linear}
    boneToModify {boneName=shoulder_pad_l}
    constraintSetup [{transformType=Rotation, targetBone {boneName=upperarm_twist_01_l}}]
    constraintWeights [1]
   }
   n94 = Get bIsSkyDiving()
   n95 = "Transform (Modify) Bone - Bone: shoulder_pad_r"!(ComponentPose=@n93, bAlphaBoolEnabled:bool=@n94, Translation="-1.435,-1.885,-0.753", Rotation="21.848,0.543,17.342", Scale="1,1,1") -> (Pose) {
    alphaInputType=Bool
    rotationMode=BMM_Additive
    rotationSpace=BCS_BoneSpace
    translationMode=BMM_Additive
    translationSpace=BCS_BoneSpace
    alphaBoolBlend {blendInTime=0.3, blendOption=Linear, blendOutTime=0.3}
    alphaScaleBiasClamp {interpSpeedDecreasing=20, interpSpeedIncreasing=20}
    boneToModify {boneName=shoulder_pad_r}
    rotation=<rot 21.848 0.543 17.342>
    translation=<-1.435 -1.885 -0.753>
   }
   n96 = "Transform (Modify) Bone - Bone: shoulder_pad_l"!(ComponentPose=@n95, bAlphaBoolEnabled:bool=@n94, Translation="1.628,-1.070,0.824", Rotation="12.776,-12.841,-7.903", Scale="1,1,1") -> (Pose) {
    alphaInputType=Bool
    rotationMode=BMM_Additive
    rotationSpace=BCS_BoneSpace
    translationMode=BMM_Additive
    translationSpace=BCS_BoneSpace
    alphaBoolBlend {blendInTime=0.3, blendOption=Linear, blendOutTime=0.3}
    boneToModify {boneName=shoulder_pad_l}
    rotation=<rot 12.776 -12.841 -7.903>
    translation=<1.628 -1.07 0.824>
   }
   n97 = "Transform (Modify) Bone - Bone: shoulder_pad_r"(
    ComponentPose=@n96
    bAlphaBoolEnabled=Get Assault rifle corr() -> (Assault rifle corr:bool)
    Translation="-2,0.178,2.103"
    Rotation="-19.544,-2.763,1.172"
    Scale="1,1,1"
   ) {
    alphaInputType=Bool
    rotationMode=BMM_Additive
    rotationSpace=BCS_BoneSpace
    translationMode=BMM_Additive
    translationSpace=BCS_BoneSpace
    alphaBoolBlend {blendInTime=0.1, blendOption=Linear, blendOutTime=0.4}
    alphaScaleBiasClamp {bInterpResult=true, interpSpeedDecreasing=8, interpSpeedIncreasing=30}
    boneToModify {boneName=shoulder_pad_r}
    rotation=<rot -19.544 -2.763 1.172>
    translation=<-2 0.178 2.103>
   }
   n98 = "Transform (Modify) Bone - Bone: shoulder_pad_l"(
    ComponentPose=@n97
    bAlphaBoolEnabled=Get Machine gun corr() -> (Machine gun corr:bool)
    Translation="1.827,-0.959,-1.602"
    Rotation="-15.769,-3.499,6.651"
    Scale="1,1,1"
   ) {
    alphaInputType=Bool
    rotationMode=BMM_Additive
    rotationSpace=BCS_BoneSpace
    translationMode=BMM_Additive
    translationSpace=BCS_BoneSpace
    alphaBoolBlend {blendInTime=0.1, blendOption=Linear, blendOutTime=0.4}
    alphaScaleBiasClamp {bInterpResult=true, interpSpeedDecreasing=8, interpSpeedIncreasing=30}
    boneToModify {boneName=shoulder_pad_l}
    rotation=<rot -15.769 -3.499 6.651>
    translation=<1.827 -0.959 -1.602>
   }
   n99 = "Transform (Modify) Bone - Bone: shoulder_pad_r"!(
    ComponentPose=@n98
    Alpha:float=Get Right pad machine gun down corr() -> (Right pad machine gun down corr:double)
    Translation="-5.689,0.769,3.433"
    Rotation="-13.063,-7.323,-0.638"
    Scale="1,1,1"
   ) -> (Pose) {
    rotationMode=BMM_Additive
    rotationSpace=BCS_BoneSpace
    translationMode=BMM_Additive
    translationSpace=BCS_BoneSpace
    alphaBoolBlend {blendInTime=0.1, blendOption=Linear, blendOutTime=0.4}
    alphaScaleBiasClamp {interpSpeedDecreasing=8, interpSpeedIncreasing=30}
    boneToModify {boneName=shoulder_pad_r}
    rotation=<rot -13.063 -7.323 -0.638>
    translation=<-5.689 0.769 3.433>
   }
   n100 = "Transform (Modify) Bone - Bone: shoulder_pad_r"(
    ComponentPose=@n99
    Alpha=Get Right pad machine gun side corr() -> (Right pad machine gun side corr:double)
    Translation="0.102,0.804,-0.202"
    Rotation="1.016,-8.315,0.408"
    Scale="1,1,1"
   ) {
    rotationMode=BMM_Additive
    rotationSpace=BCS_BoneSpace
    translationMode=BMM_Additive
    translationSpace=BCS_BoneSpace
    alphaBoolBlend {blendInTime=0.1, blendOption=Linear, blendOutTime=0.4}
    alphaScaleBiasClamp {interpSpeedDecreasing=8, interpSpeedIncreasing=30}
    boneToModify {boneName=shoulder_pad_r}
    rotation=<rot 1.016 -8.315 0.408>
    translation=<0.102 0.804 -0.202>
   }
   n101 = "Transform (Modify) Bone - Bone: shoulder_pad_r"!(
    ComponentPose=@n100
    bAlphaBoolEnabled:bool=Get Rocket launcher corr() -> (Rocket launcher corr:bool)
    Translation="-0.903,0.000,0.634"
    Rotation="-4.853,-1.731,0.148"
    Scale="1,1,1"
   ) -> (Pose) {
    alphaInputType=Bool
    rotationMode=BMM_Additive
    rotationSpace=BCS_BoneSpace
    translationMode=BMM_Additive
    translationSpace=BCS_BoneSpace
    alphaBoolBlend {blendInTime=0.1, blendOption=Linear, blendOutTime=0.4}
    alphaScaleBiasClamp {bInterpResult=true, interpSpeedDecreasing=8, interpSpeedIncreasing=30}
    boneToModify {boneName=shoulder_pad_r}
    rotation=<rot -4.853 -1.731 0.148>
    translation {x=-0.903, z=0.634}
   }
   n102 = Get Right pad emote corr() -> (Right pad emote corr:double)
   n103 = "Constraint - shoulder_pad_r (RR)"!(ComponentPose=@n101, Alpha=@n102, ConstraintWeights_0=1, ConstraintWeights_1:float=0) -> (Pose) {
    alphaBoolBlend {blendInTime=0.2, blendOption=Linear}
    alphaScaleBiasClamp {bInterpResult=true}
    boneToModify {boneName=shoulder_pad_r}
    constraintSetup [{transformType=Rotation, targetBone {boneName=upperarm_twist_01_r}}, {transformType=Rotation, targetBone {boneName=clavicle_r}}]
    constraintWeights [1,1]
   }
   n104 = "Transform (Modify) Bone - Bone: shoulder_pad_r"!(ComponentPose=@n103, Alpha:float=@n102, Translation="-1.133,0.460,0.968", Rotation="-16.641,13.565,-0.413", Scale="1,1,1") -> (Pose) {
    rotationMode=BMM_Additive
    rotationSpace=BCS_BoneSpace
    translationMode=BMM_Additive
    translationSpace=BCS_BoneSpace
    alphaBoolBlend {blendInTime=0.3, blendOption=Linear}
    alphaScaleBiasClamp {bInterpResult=true, interpSpeedDecreasing=15}
    boneToModify {boneName=shoulder_pad_r}
    rotation=<rot -16.641 13.565 -0.413>
    translation=<-1.133 0.46 0.968>
   }
   n105 = Get Left pad emote corr() -> (Left pad emote corr:double)
   n106 = "Constraint - shoulder_pad_l (RR)"!(ComponentPose=@n104, Alpha=@n105, ConstraintWeights_0=1, ConstraintWeights_1:float=0) -> (Pose) {
    alphaBoolBlend {blendInTime=0.2, blendOption=Linear}
    boneToModify {boneName=shoulder_pad_l}
    constraintSetup [{transformType=Rotation, targetBone {boneName=upperarm_twist_01_l}}, {transformType=Rotation, targetBone {boneName=clavicle_l}}]
    constraintWeights [1,1]
   }
   n107 = "Transform (Modify) Bone - Bone: shoulder_pad_l"!(ComponentPose=@n106, Alpha:float=@n105, Translation="2.224,-0.461,0.593", Rotation="-12.885,14.262,6.207", Scale="1,1,1") -> (Pose) {
    rotationMode=BMM_Additive
    rotationSpace=BCS_BoneSpace
    translationMode=BMM_Additive
    translationSpace=BCS_BoneSpace
    alphaBoolBlend {blendInTime=0.3, blendOption=Linear}
    alphaScaleBiasClamp {bInterpResult=true, interpSpeedDecreasing=20}
    boneToModify {boneName=shoulder_pad_l}
    rotation=<rot -12.885 14.262 6.207>
    translation=<2.224 -0.461 0.593>
   }
   Save cached pose 'Shoulder pads cache'(Pose=Component To Local(ComponentPose=@n107)) {cacheName=Shoulder pads cache}
  ]}
 EventGraph {nodes [
   Event BlueprintUpdateAnimation()
   > n1 = Sequence!(execute) -> (then_0, then_1, then_2:Exec)
   n2 = Gravity Override State Function DEPRECATED(
    execute=@n1.then_1
    CharacterPartAnimInstance=Self-Reference()
    GameTimeInSeconds=GetGameTimeInSeconds()
    GravityOverrideParams=Get Gravity Override Params()
   )
   > Set Gravity Override(Gravity Override=@n2)
   n3 = Get bIsMoving2D()
   > "Set Walking/running  alpha"("Walking/running  alpha":bool=@n3) -> (Output_Get:bool)
   n4 = Get PawnMovementDirectionAzimuth() -> (PawnMovementDirectionAzimuth:float)
   n5 = OR Boolean(A="float > float"(A=@n4, B=135), B="float < float"(A=@n4, B=-135))
   n6 = Get bIsMoving2D()
   n7 = Get PawnMovementDirectionAzimuth()
   n8 = SelectVector(
    A="-1,-1,-2"
    B=SelectVector(A="0, 0, 0", B="-1,-1,2", bPickA=AND Boolean(A=@n5, B=@n6))
    bPickA=AND Boolean!(A=@n3, B="float > float"(A=@n7, B=-180), C:bool="float < float"(A=@n7, B=-45)) -> (ReturnValue)
   )
   n9 = Get bIsCrouching()
   n10 = AND Boolean(A=AND Boolean!(A=NOT Boolean(A=@n9), B=@n3) -> (ReturnValue), B="float > float"(A=@n7, B=45), C="float < float"(A=@n7, B=135))
   n11 = "float > float"(A=@n7, B=-45)
   n12 = "float < float"(A=@n7, B=45)
   > Set Front cloth corr(Front cloth corr:Vector=SelectVector(A="-1,-1,-1", B=SelectVector(A="-1,-1,2", B=@n8, bPickA=@n10), bPickA=AND Boolean!(A=@n11, B=@n12, C:bool=@n3) -> (ReturnValue))) -> (Output_Get:Vector)
   n13 = Get PawnMovementDirectionAzimuth()
   n14 = AND Boolean!(A=Get bIsMoving2D(), B=OR Boolean(A="float > float"(A=@n13, B=135), B="float < float"(A=@n13, B=-135))) -> (ReturnValue)
   > Set Front cloth rot corr(Front cloth rot corr:double=SelectFloat(A=-30, B=SelectFloat(A=-5, B=-20, bPickA=@n14), bPickA=Get bIsPlayingEmote())) -> (Output_Get:double)
   n15 = Get bIsKneeSliding()
   > Set Crouching alpha(Crouching alpha:bool=AND Boolean(A=Get bIsCrouchSprinting(), B=NOT Boolean(A=@n15))) -> (Output_Get:bool)
   n16 = Get bIsMoving2D()
   > Set Crouching idle alpha(Crouching idle alpha:bool=AND Boolean(A=Get bIsCrouching(), B=NOT Boolean(A=@n16))) -> (Output_Get:bool)
   > Set Is melee while crouch(Is melee while crouch:bool=AND Boolean(A=Get bIsPlayingMeleeAnim(), B=Get bIsCrouching())) -> (Output_Get:bool)
   > Set Crouching corr(Crouching corr:bool=AND Boolean(A=Get bIsCrouching(), B=Get bIsMoving2D())) -> (Output_Get:bool)
   > n17 = Set On Ground Anim Dynamics Alpha(On Ground Anim Dynamics Alpha:double=SelectFloat(A=0, B=1, bPickA=Get bIsSkyDiving())) -> (Output_Get:double)
   > Set Skydiving Anim Dynamics Alpha(Skydiving Anim Dynamics Alpha:double=SelectFloat(A=1, B=0, bPickA=Get bIsSkyDiving())) -> (Output_Get:double)
   > Set On Ground Chain End Element Anim Dynamics Alpha(On Ground Chain End Element Anim Dynamics Alpha:double=SelectFloat(A=0.2, B=@n17, bPickA=Get bIsTacticalSprinting())) -> (Output_Get:double)
   > "Set Is Sprint Jumping "("Is Sprint Jumping ":bool=AND Boolean(A=Get bIsTacticalSprinting(), B=Get bIsJumping())) -> (Output_Get:bool)
   > n18 = Cast To FortPlayerAnimInstance(Object=GetAnimInstance(self=GetOwnerSkeletalMeshComponent()))
   > Set Assault rifle corr(Assault rifle corr:bool=false) -> (Output_Get:bool)
   > Set Machine gun corr(Machine gun corr:bool=false) -> (Output_Get:bool)
   > Set Rocket launcher corr(Rocket launcher corr:bool=false) -> (Output_Get:bool)
   > Set Dual Wield corr(Dual Wield corr:bool=false) -> (Output_Get:bool)
   > n19 = Switch on EFortWeaponCoreAnimation!(execute, Selection:EFortWeaponCoreAnimation Enum=Get WeaponCoreAnim(self=@n18)) -> (
    Melee:Exec
    Pistol:Exec
    Shotgun:Exec
    PaperBlueprint:Exec
    Rifle:Exec
    MeleeOneHand:Exec
    MachinePistol:Exec
    RocketLauncher:Exec
    GrenadeLauncher:Exec
    GoingCommando:Exec
    AssaultRifle:Exec
    TacticalShotgun:Exec
    SniperRifle:Exec
    TrapPlacement:Exec
    ShoulderLauncher:Exec
    AbilityDecoTool:Exec
    Crossbow:Exec
    C4:Exec
    RemoteControl:Exec
    DualWield:Exec
    AR_BullPup:Exec
    AR_ForwardGrip:Exec
    MedPackPaddles:Exec
    SMG_P90:Exec
    AR_DrumGun:Exec
    Consumable_Small:Exec
    Consumable_Large:Exec
    Balloon:Exec
    MountedTurret:Exec
    CreativeTool:Exec
    ExplosiveBow:Exec
    AshtonIndigo:Exec
    AshtonChicago:Exec
    MeleeDualWield:Exec
    Unarmed:Exec
   )
   Set Assault rifle corr(execute=@n19.AssaultRifle, Assault rifle corr=true)
   Set Dual Wield corr(execute=@n19.DualWield, Dual Wield corr=true)
   Set Machine gun corr(execute=@n19.GoingCommando, Machine gun corr=true)
   n20 = Branch(execute=@n19.RocketLauncher, Condition=Get bIsTargeting())
   > Set Rocket launcher corr(Rocket launcher corr=true)
   Set Rocket launcher corr(execute=@n20.else, Rocket launcher corr=false)
   n21 = Set Skeletal Mesh(execute=@n31.then, Skeletal Mesh:Skeletal Mesh Component Object Reference=GetPartSkeletalMeshComponent()) -> (Output_Get:Skeletal Mesh Component Object Reference)
   n22 = float + float(
    A=GetAlphaFromReferenceTransformDelta(SkeletalMeshComponent=@n21, BoneName=upperarm_r, BaseName=clavicle_r, InputRange=(X=42,Y=110), TwistAxis="1,0,0")
    B=GetAlphaFromReferenceTransformDelta(SkeletalMeshComponent=@n21, BoneName=upperarm_r, BaseName=clavicle_r, DeltaComponentType=EulerY, InputRange=(X=-36,Y=29), TwistAxis="1,0,0")
   )
   Set Right Pad alpha(execute=@n1.then_0, Right Pad alpha:double="Clamp (Float)"(Value=float / float(A=@n22, B=2), Min=0, Max=1)) -> (Output_Get:double)
   n23 = GetAlphaFromReferenceTransformDelta(SkeletalMeshComponent=@n21, BoneName=upperarm_r, BaseName=clavicle_r, DeltaComponentType=EulerY, InputRange=(X=20,Y=50), TwistAxis="1,0,0")
   n24 = "Clamp (Float)"(
    Value=float + float(
     A=@n23
     B=GetAlphaFromReferenceTransformDelta(SkeletalMeshComponent=@n21, BoneName=upperarm_r, BaseName=clavicle_r, InputRange=(X=-10,Y=30), OutputRange=(X=1,Y=0), TwistAxis="1,0,0")
    )
    Min=0
    Max=1
   )
   n25 = "float > float"(A=@n23.RawOutput, B=20)
   > Set Right pad emote corr(Right pad emote corr:double=SelectFloat(A=0.3, B=SelectFloat(A=@n24, B=0, bPickA=@n25), bPickA=Get bIsSkyDiving())) -> (Output_Get:double)
   n26 = GetAlphaFromReferenceTransformDelta(SkeletalMeshComponent=@n21, BoneName=upperarm_l, BaseName=clavicle_l, DeltaComponentType=EulerY, InputRange=(X=20,Y=50), TwistAxis="1,0,0")
   n27 = "Clamp (Float)"(
    Value=float + float(
     A=@n26
     B=GetAlphaFromReferenceTransformDelta(SkeletalMeshComponent=@n21, BoneName=upperarm_l, BaseName=clavicle_l, InputRange=(X=10,Y=-30), TwistAxis="1,0,0")
    )
    Min=0
    Max=1
   )
   n28 = Get Dual Wield corr() -> (Dual Wield corr:bool)
   n29 = OR Boolean(
    A=AND Boolean(A="float > float"(A=@n26, B=20), B=NOT Boolean(A=@n28))
    B=AND Boolean(A="float > float"(A=@n26, B=20), B=Get bIsPlayingEmote())
   )
   > Set Left pad emote corr(Left pad emote corr:double=SelectFloat(A=0.3, B=SelectFloat(A=@n27, B=0, bPickA=@n29), bPickA=Get bIsSkyDiving())) -> (Output_Get:double)
   n30 = float + float(
    A=GetAlphaFromReferenceTransformDelta(SkeletalMeshComponent=@n21, BoneName=upperarm_l, BaseName=clavicle_l, DeltaComponentType=EulerY, InputRange=(X=-36,Y=45), TwistAxis="1,0,0")
    B=GetAlphaFromReferenceTransformDelta(SkeletalMeshComponent=@n21, BoneName=upperarm_l, BaseName=clavicle_l, InputRange=(X=10,Y=60), TwistAxis="1,0,0")
   )
   > Set Left Pad alpha(Left Pad alpha:double="Clamp (Float)"(Value=float / float(A=@n30, B=2), Min=0, Max=0.95)) -> (Output_Get:double)
   > Set Right pad machine gun down corr(Right pad machine gun down corr:double=SelectFloat(
     A=GetAlphaFromReferenceTransformDelta(SkeletalMeshComponent=@n21, BoneName=clavicle_r, BaseName=spine_05, DeltaComponentType=EulerY, InputRange=(X=8,Y=50), TwistAxis="1,0,0")
     B=0
     bPickA=Get Machine gun corr()
    )) -> (Output_Get:double)
   > Set Right pad machine gun side corr(Right pad machine gun side corr:double=SelectFloat(
     A=GetAlphaFromReferenceTransformDelta(SkeletalMeshComponent=@n21, BoneName=upperarm_r, BaseName=clavicle_r, InputRange=(X=-10,Y=10), OutputRange=(X=1,Y=0), TwistAxis="1,0,0")
     B=0
     bPickA=Get Machine gun corr()
    )) -> (Output_Get:double)
   n31 = Event OnCharacterCustomizationLoaded()
  ]}
]
variables [
 Assault rifle corr {type=boolean}
 Crouching alpha {type=boolean}
 Crouching corr {type=boolean}
 Crouching idle alpha {type=boolean}
 Dual Wield corr {type=boolean}
 Front cloth corr {type=Vector}
 Front cloth rot corr {type=number}
 Gravity Override {type=Vector}
 Gravity Override Params {
  type=GravityOverrideParamsStruct
  default {rail_Grinding {gravityOverride=<0 -450 -500>}, skydive_DivingDown {windAmplitude=<150 600 200>, windFrequency=<2 18 20>}}
 }
 Head look down alpha {type=number}
 Is melee while crouch {type=boolean}
 "Is Sprint Jumping " {type=boolean}
 Left Pad alpha {type=number}
 Left pad emote corr {type=number}
 Machine gun corr {type=boolean}
 On Ground Anim Dynamics Alpha {type=number}
 On Ground Chain End Element Anim Dynamics Alpha {type=number}
 Right Pad alpha {type=number}
 Right pad emote corr {type=number}
 Right pad machine gun down corr {type=number}
 Right pad machine gun side corr {type=number}
 Rocket launcher corr {type=boolean}
 Skeletal Mesh {type=SkeletalMeshComponent}
 Skydiving Anim Dynamics Alpha {type=number}
 "Walking/running  alpha" {type=boolean}
]
