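# Material function for the TalonHime backpack that creates animated emissive smoke seeming to sit beneath the surface, optionally added to incoming emissive.
# Tiled UVs from a selectable channel are panned, optionally flow-directed, and warped by a maskable, time-driven sine distortion from a smoke texture.
# Bump offset gives parallax depth, with a mask blending offset and flat samples.
# Brightness remap, tint, invertible Fresnel and a TechArt mask channel shape the result.
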MaterialFunction /Game/Accessories/FORT_Backpacks/Backpack_F_MED_TalonHime/Materials/MF_SmokeBeneath
graph {nodes [
  n1 = FunctionInput() {bUsePreviewValueAsDefault=true, inputName=EmissiveInput}
  n2 = Multiply(
   A=ScalarParameter() {group=SmokeBeneath, parameterName=Smoke_OffsetIntensity}
   B=AppendVector(
    A=ScalarParameter() {group=SmokeBeneath, parameterName=Smoke_UVOffsetU}
    B=ScalarParameter() {group=SmokeBeneath, parameterName=Smoke_UVOffsetV}
   )
  )
  n3 = TextureCoordinate()
  n4 = TextureCoordinate() {coordinateIndex=1}
  n5 = Round(None=ScalarParameter() {group=SmokeBeneath, parameterName=Smoke_UseSecondUVChannelForPanning})
  n6 = Multiply(
   A=NamedRerouteDeclaration(None=LinearInterpolate(A=@n3, B=@n4, Alpha=@n5)) {name=CachedBlendUV, nodeColor=<rgba 0.3359 0 1 1>}
   B=ScalarParameter() {defaultValue=1.5, group=SmokeBeneath, parameterName=Smoke_Tilling}
  )
  n7 = TextureSampleParameter2D() {
   group=SmokeBeneath
   parameterName=Smoke_BumpOffsetMask
   samplerType=SAMPLERTYPE_LinearColor
   texture=/Game/Accessories/FORT_Backpacks/Backpack_F_MED_TalonHime/Textures/T_Smokemap_BumpMask
  }
  n8 = LinearInterpolate(
   A=Constant()
   B=NamedRerouteUsage()
   Alpha=StaticSwitchParameter(
    True=StaticComponentMaskParameter(None=@n7.RGB) {defaultR=true, group=SmokeBeneath, parameterName=DistortMaskChannel}
    False=Constant() {r=1}
   ) {group=SmokeBeneath, parameterName=Smoke_UseDistortMask}
  )
  n9 = Constant()
  n10 = ScalarParameter() {group=SmokeBeneath, parameterName=Smoke_GradientSpeed_X}
  n11 = ScalarParameter() {defaultValue=0.022, group=SmokeBeneath, parameterName=Smoke_GradientSpeed_Y}
  n12 = Panner(
   Coordinate=Add(A=Add(A=@n2, B=@n6), B=StaticSwitchParameter(True=@n8, False=@n9) {defaultValue=true, group=SmokeBeneath, parameterName=Smoke_UseDistort})
   Speed=LinearInterpolate(
    A=AppendVector(A=@n10, B=@n11)
    B=FunctionInput() {bUsePreviewValueAsDefault=true, inputName=FlowSpeed, inputType=FunctionInput_Vector2, sortPriority=2}
    Alpha=ScalarParameter() {group=SmokeBeneath, parameterName=CustomizedFlowDirection}
   )
  )
  n13 = TextureSampleParameter2D(UVs=@n12) {
   group=SmokeBeneath
   parameterName=SmokeMap
   samplerSource=SSM_Wrap_WorldGroupSettings
   samplerType=SAMPLERTYPE_LinearColor
   texture=/Game/Accessories/FORT_Backpacks/Backpack_F_MED_TalonHime/Textures/T_Smokemap
  }
  n14 = ScalarParameter() {defaultValue=-1, group=SmokeBeneath, parameterName=Smoke_Bump Offset Height01}
  n15 = TextureSampleParameter2D(UVs=StaticSwitchParameter(True=BumpOffset(Coordinate=@n12, Height=@n14), False=@n12) {defaultValue=true, group=SmokeBeneath, parameterName=use Bump Offset on Smoke}) {
   group=SmokeBeneath
   parameterName=SmokeMap
   samplerSource=SSM_Wrap_WorldGroupSettings
   samplerType=SAMPLERTYPE_LinearColor
   texture=&T_Smokemap
  }
  n16 = TextureSampleParameter2D() {
   group=SmokeBeneath
   parameterName=Smoke_BumpOffsetMask
   samplerType=SAMPLERTYPE_LinearColor
   texture=&T_Smokemap_BumpMask
  }
  n17 = Power(
   Base=StaticComponentMaskParameter(None=@n16.RGB) {defaultR=true, group=SmokeBeneath, parameterName=BumpMaskChannel}
   Exp=ScalarParameter() {defaultValue=0.5, group=SmokeBeneath, parameterName=Smoke_Bump Power}
  )
  n18 = LinearInterpolate(
   A=ScalarParameter() {defaultValue=0.15, group=SmokeBeneath, parameterName=Smoke_BrightnessMIn}
   B=ScalarParameter() {defaultValue=2, group=SmokeBeneath, parameterName=Smoke_BrightnessMax}
   Alpha=StaticSwitchParameter(True=LinearInterpolate(A=@n13.B, B=@n15.B, Alpha=@n17), False=@n15.B) {defaultValue=true, group=SmokeBeneath, parameterName=UseBumpOffsetMask}
  )
  n19 = VectorParameter() {group=SmokeBeneath, parameterName=Smoke_Emissive, defaultValue=<rgba 1 1 1 1>}
  n20 = Saturate(None=Fresnel(ExponentIn=ScalarParameter() {defaultValue=4, group=SmokeBeneath, parameterName=Smoke_Emissive_Fresnel Exponent}) {baseReflectFraction=0})
  n21 = OneMinus(None=@n20)
  n22 = ScalarParameter() {group=SmokeBeneath, parameterName=Smoke_Emissive_Invert Fresnel}
  n23 = Multiply(
   A=Multiply(A=Multiply(A=@n18, B=@n19.RGB), B=LinearInterpolate(A=@n21, B=@n20, Alpha=@n22))
   B=StaticComponentMaskParameter(None=FunctionInput() {bUsePreviewValueAsDefault=true, inputName=TechArtMaskInput}) {defaultR=true, group=SmokeBeneath, parameterName=SmokeBeneathChannel}
  )
  FunctionOutput(None=StaticSwitchParameter(True=Add(A=@n1, B=@n23), False=@n23) {defaultValue=true, group=SmokeBeneath, parameterName=Draw Emissive Above Smoke})
  n24 = TextureSampleParameter2D(UVs=Multiply(A=NamedRerouteUsage(), B=ScalarParameter() {defaultValue=1, group=SmokeBeneath, parameterName=Smoke_DistortTiling})) {
   group=SmokeBeneath
   parameterName=Smoke_DistortUV
   samplerSource=SSM_Wrap_WorldGroupSettings
   samplerType=SAMPLERTYPE_LinearColor
   texture=&T_Smokemap
  }
  n25 = Subtract(A=Multiply(A=ComponentMask(None=@n24.RGB) {g=true, r=true}) {constB=2})
  n26 = ComponentMask(None=@n25) {r=true}
  n27 = Constant()
  n28 = StaticSwitchParameter(
   True=AppendVector(A=Constant(), B=ComponentMask(None=@n25) {g=true})
   False=StaticSwitchParameter(True=AppendVector(A=@n26, B=@n27), False=@n25) {defaultValue=true, group=SmokeBeneath, parameterName=Smoke_UseXwave}
  ) {group=SmokeBeneath, parameterName=Smoke_UseYwave}
  n29 = Time()
  n30 = ScalarParameter() {defaultValue=0.2, group=SmokeBeneath, parameterName=Smoke_DistortTimeScale}
  n31 = StaticSwitchParameter(True=Sine(None=Multiply(A=@n29, B=@n30) {constB=0.02}), False=Constant() {r=1}) {defaultValue=true, group=SmokeBeneath, parameterName=Smoke_UseTimeDriver}
  NamedRerouteDeclaration(None=Multiply(A=Multiply(A=@n28, B=@n31), B=ScalarParameter() {defaultValue=0.05, group=SmokeBeneath, parameterName=Smoke_DistortIntensity}) {constB=0.5}) {name=CachedDistortUV, nodeColor=<rgba 0.3594 1 0 1>}
 ]}
